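# One-shot Niagara burst for the Citadel emote's armor-off moment, flashing light and throwing debris from the character.
# Short-lived flare, corona and glow sprites stack at a shared System.FlarePos, while bone-attached fake-light sprites and a real light sample the user skeletal mesh.
# Mesh fragments burst from that mesh with drag and curl noise, alongside GPU sparks from an offset sphere.
# Time-of-day scalars dim most flares and the light, with Dark_Glow also adjusting for front end.
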NiagaraSystem /BRCosmetics/Animation/Game/MainPlayer/Emotes/Citadel/FX/NS_Emote_Citadel_ArmorOff_Burst
dependencies {
 usedDataInterfaces [&NiagaraDataInterfaceColorCurve, &NiagaraDataInterfaceCurve]
 usedDynamicInputs []
 usedModules [
  &self:SystemSpawnScript.SystemScriptSource.SystemScriptGraph.NiagaraNodeAssignment_0.SetVariables_#1
  &SystemState
  &self:SystemSpawnScript.SystemScriptSource.SystemScriptGraph.NiagaraNodeAssignment_1.SetVariables_#2
 ]
}
emitters {
 CoreFlare {
  emitterUpdateScript [
   &EmitterState()
   &SpawnBurst_Instantaneous(Spawn Count=1, Use Spawn Probability=false)
  ]
  particleSpawnScript [
   /Niagara/Modules/Spawn/Initialization/InitializeParticle(
    Color=<rgba 5 3.6875 1.25 0.15>
    Lifetime=0.35
    Module.Write Color=true
    Module.Write Lifetime=true
    Module.Write Mass=true
    Module.Write Material Random=true
    Module.Write Position=false
    Module.Write Scale=false
    Module.Write SpriteRotation=false
    Module.Write SpriteSize=true
    Position=&SimulationPosition()
    Sprite Size=&Vector2DFromFloat(Value=1500)
   )
   &CameraOffset(Camera Offset Amount=-30)
   &DynamicMaterialParameters(
    Index 0 Param 1=0.1
    Index 0 Param 2=1
    Index 0 Param 3=1
    Index 0 Param 4=1
    Module.Param0WriteEnabled=true
    Module.Param1WriteEnabled=true
    Module.Param2WriteEnabled=true
    Module.Param3WriteEnabled=true
   )
   &SphereLocation(
    MaskSpawnGroup=false
    Offset=$System.FlarePos
    Override Seed=false
    Sphere Radius=0
    Surface Only=false
    UseNonUniformScale=false
    UseOffset=true
    UseSphereOrientation=false
   )
  ]
  particleUpdateScript [
   &ParticleState()
   &SolveForcesAndVelocity(Clamp Velocity=false, Limit Acceleration=false)
   &ScaleColor(Scale Alpha=0.7, ScaleA=true, ScaleRGB=true)
   &ScaleSpriteSize(Scale Factor=&Vector2DFromFloat(Value=&FloatFromCurve(FloatCurve=DataInterface() {propertyValues {Curve {keys [
          {arriveTangent=3310.2024, interpMode=RCIM_Cubic, time=0, value=0}
          {arriveTangent=-15.1945, interpMode=RCIM_Cubic, leaveTangent=-3.0716, tangentMode=RCTM_Break, time=0.15, value=1}
          {interpMode=RCIM_Cubic, leaveTangent=-1.6445, time=1, value=0}
         ]}}})))
   &ScaleColor(Scale Alpha=$System.TOD, ScaleA=true, ScaleRGB=false) {module=ScaleColor001}
  ]
  rendererClasses [&NiagaraSpriteRendererProperties]
  renderers [{rendererClass=&NiagaraSpriteRendererProperties}]
 }
 CoreFlare_MedHigh {
  emitterUpdateScript [
   &EmitterState()
   &SpawnBurst_Instantaneous(Spawn Count=1, Use Spawn Probability=false)
  ]
  particleSpawnScript [
   &InitializeParticle(
    Color=<rgba 5 4.125 2.5 0.1>
    Lifetime=0.35
    Module.Write Color=true
    Module.Write Lifetime=true
    Module.Write Mass=true
    Module.Write Material Random=true
    Module.Write Position=false
    Module.Write Scale=false
    Module.Write SpriteRotation=false
    Module.Write SpriteSize=true
    Position=&SimulationPosition()
    Sprite Size=&Vector2DFromFloat(Value=1500)
   )
   &CameraOffset(Camera Offset Amount=-20)
   &DynamicMaterialParameters(
    Index 0 Param 1=0.1
    Index 0 Param 2=1
    Index 0 Param 3=1
    Index 0 Param 4=1
    Module.Param0WriteEnabled=true
    Module.Param1WriteEnabled=true
    Module.Param2WriteEnabled=true
    Module.Param3WriteEnabled=true
   )
   &SphereLocation(
    MaskSpawnGroup=false
    Offset=$System.FlarePos
    Override Seed=false
    Sphere Radius=0
    Surface Only=false
    UseNonUniformScale=false
    UseOffset=true
    UseSphereOrientation=false
   )
  ]
  particleUpdateScript [
   &ParticleState()
   &SolveForcesAndVelocity(Clamp Velocity=false, Limit Acceleration=false)
   &ScaleColor(Scale Alpha=0.7, ScaleA=true, ScaleRGB=true)
   &ScaleSpriteSize(Scale Factor=&Vector2DFromFloat(Value=&FloatFromCurve(FloatCurve=DataInterface() {propertyValues {Curve {keys [
          {arriveTangent=3310.2024, interpMode=RCIM_Cubic, time=0, value=0}
          {arriveTangent=0.7273, interpMode=RCIM_Cubic, leaveTangent=0.7273, time=0.1, value=0.3}
          {arriveTangent=-0.3333, interpMode=RCIM_Cubic, leaveTangent=-0.3333, time=0.55, value=0.4}
          {interpMode=RCIM_Cubic, leaveTangent=-1.6445, time=1, value=0}
         ]}}})))
   &ScaleColor(Scale Alpha=$System.TOD, ScaleA=true, ScaleRGB=false) {module=ScaleColor001}
  ]
  rendererClasses [&NiagaraSpriteRendererProperties]
  renderers [{rendererClass=&NiagaraSpriteRendererProperties}]
 }
 Corona {
  emitterUpdateScript [
   &EmitterState()
   &SpawnBurst_Instantaneous(Spawn Count=1, Spawn Time=0.05, Use Spawn Probability=false)
  ]
  particleSpawnScript [
   &InitializeParticle(
    Color=<rgba 6 4.8 1.5 0.35>
    Lifetime=0.25
    Module.Write Color=true
    Module.Write Lifetime=true
    Module.Write Mass=true
    Module.Write Material Random=true
    Module.Write Position=false
    Module.Write Scale=false
    Module.Write SpriteRotation=false
    Module.Write SpriteSize=true
    Position=&SimulationPosition()
    Sprite Size=&Vector2DFromFloat(Value=350)
   )
   &CameraOffset(Camera Offset Amount=1)
   &SphereLocation(
    MaskSpawnGroup=false
    Offset=$System.FlarePos
    Override Seed=false
    Sphere Radius=0
    Surface Only=false
    UseNonUniformScale=false
    UseOffset=true
    UseSphereOrientation=false
   )
  ]
  particleUpdateScript [
   &ParticleState()
   &SolveForcesAndVelocity(Clamp Velocity=false, Limit Acceleration=false)
   &DynamicMaterialParameters(
    Index 0 Param 1=&FloatFromCurve(FloatCurve=DataInterface() {propertyValues {Curve {keys [{arriveTangent=-2.7401, interpMode=RCIM_Cubic, leaveTangent=-2.7401, tangentMode=RCTM_User, time=0, value=2}, <at 1 1 RCIM_Cubic>]}}})
    Index 0 Param 2=0.4
    Index 0 Param 3=0.6
    Index 0 Param 4=1
    Module.Param0WriteEnabled=true
    Module.Param1WriteEnabled=true
    Module.Param2WriteEnabled=true
    Module.Param3WriteEnabled=true
   )
   &ScaleColor(
    Scale Alpha=&FloatFromCurve(FloatCurve=DataInterface() {propertyValues {Curve {keys [
         <at 0 0 RCIM_Cubic>
         {arriveTangent=-0.5786, interpMode=RCIM_Cubic, leaveTangent=-0.5786, tangentMode=RCTM_User, time=0.15, value=0.6}
         <at 1 0 RCIM_Cubic>
        ]}}})
    Scale RGB=<2 2 2>
    ScaleA=true
    ScaleRGB=true
   )
   &ScaleSpriteSize(Scale Factor=&Vector2DFromFloat(Value=&FloatFromCurve(FloatCurve=DataInterface() {propertyValues {Curve {keys [
          {arriveTangent=3310.2024, interpMode=RCIM_Cubic, time=0, value=0.1}
          {arriveTangent=1.4167, interpMode=RCIM_Cubic, leaveTangent=1.4167, time=0.4, value=0.8}
          {arriveTangent=0.3333, interpMode=RCIM_Cubic, leaveTangent=0.3333, time=0.6, value=0.95}
          {interpMode=RCIM_Cubic, leaveTangent=-1.6445, time=1, value=1}
         ]}}})))
   &ScaleColor(Scale Alpha=$System.TOD, ScaleA=true, ScaleRGB=false) {module=ScaleColor001}
  ]
  rendererClasses [&NiagaraSpriteRendererProperties]
  renderers [{rendererClass=&NiagaraSpriteRendererProperties}]
 }
 Dark_Glow {
  emitterUpdateScript [
   &EmitterState()
   &SpawnBurst_Instantaneous(Spawn Count=1, Use Spawn Probability=false)
  ]
  particleSpawnScript [
   /Niagara/Modules/Spawn/Initialization/V2/InitializeParticle(
    Color=<rgba 0.0781 0.0566 0.0264 1>
    Lifetime=0.22
    Sprite Size Mode=::NewEnumerator3
    Uniform Sprite Size=300
    UsePositionOffset=false
    Write Material Random=true
   ) {module=InitializeParticle001}
   &DynamicMaterialParameters(
    Index 0 Param 1=1
    Index 0 Param 2=1
    Index 0 Param 3=1
    Index 0 Param 4=1
    Module.Param0WriteEnabled=true
    Module.Param1WriteEnabled=true
    Module.Param2WriteEnabled=true
    Module.Param3WriteEnabled=true
   )
   &SystemLocation(MaskSpawnGroup=false, Offset=$System.FlarePos)
  ]
  particleUpdateScript [
   &CameraOffset(Camera Offset Amount=-20)
   &ParticleState()
   &SolveForcesAndVelocity(Clamp Velocity=false, Limit Acceleration=false)
   &ScaleColor(
    Scale Alpha=&Multiply_Float(
     A=&FloatFromCurve(
      FloatCurve=DataInterface() {propertyValues {Curve {keys [<at 0 1>,<at 1 0>]}}}
      Scale Curve=&FN_FrontEndFloatModifier(Gameplay=1, Front End=0.5)
     )
     B=&FNTimeOfDayScalar(Night Scale=0.3, Morning Scale=0.05, Day Scale=0.5, Evening Scale=0.05)
    )
    ScaleA=true
    ScaleRGB=true
   )
   &ScaleSpriteSize(Scale Factor=&Vector2DFromFloat(Value=&FloatFromCurve(FloatCurve=DataInterface() {propertyValues {Curve {keys [<at 0 1>,<at 1 3>]}}})))
  ]
  rendererClasses [&NiagaraSpriteRendererProperties]
  renderers [{rendererClass=&NiagaraSpriteRendererProperties}]
 }
 FakeLight {
  emitterUpdateScript [
   &EmitterState()
   &SpawnBurst_Instantaneous(Spawn Count=13, Use Spawn Probability=false, UseLoopCountLimit=false)
  ]
  particleSpawnScript [
   &InitializeParticle(
    Color=<rgba 8 6.2667 4 1>
    Lifetime=0.25
    Sprite Size Mode=::NewEnumerator3
    Uniform Sprite Size=100
    UsePositionOffset=false
    Write Material Random=true
   )
   &FN_SetSpawnGroup(Normalize Spawn Group=true, Number of Spawn Groups=13, Particles Per Group=1)
   &CameraOffset(Camera Offset Amount=50)
  ]
  particleUpdateScript [
   &ParticleState()
   &ScaleColor(
    Scale Alpha=&FloatFromCurve(FloatCurve=DataInterface() {propertyValues {Curve {keys [
         {arriveTangent=9.0725, interpMode=RCIM_Cubic, leaveTangent=9.0725, tangentMode=RCTM_User, time=0, value=0}
         {arriveTangent=-2.5249, interpMode=RCIM_Cubic, leaveTangent=-2.5249, tangentMode=RCTM_User, time=0.05, value=0.9847}
         {arriveTangent=-0.0405, interpMode=RCIM_Cubic, leaveTangent=-0.0405, tangentMode=RCTM_User, time=1, value=0}
        ]}}})
    Scale RGB=&MakeVectorFromLinearColorRGB(Color=&ColorFromCurve(ColorCurve=DataInterface() {propertyValues {
        AlphaCurve {keys [<at 0 1 RCIM_Cubic>]}
        BlueCurve {keys [<at 0 6.4>,<at 0.1743 2.4>]}
        GreenCurve {keys [<at 0 7.3067>,<at 0.1743 4.44>]}
        RedCurve {keys [<at 0 8>,<at 0.1743 6>]}
       }}))
    ScaleA=true
    ScaleRGB=false
   )
   &SkeletalMeshLocation(Bone / Socket Index=$Particles.SpawnGroup, Bone Sampling Mode=::NewEnumerator4, MaskSpawnGroup=false, Skeletal Mesh=$User.Skeletal Mesh)
  ]
  rendererClasses [&NiagaraSpriteRendererProperties]
  renderers [{rendererClass=&NiagaraSpriteRendererProperties}]
 }
 Fragments_Burst {
  emitterUpdateScript [
   &EmitterState()
   &SpawnBurst_Instantaneous(Spawn Count=35, Use Spawn Probability=false, UseLoopCountLimit=false) {module=SpawnBurst_Instantaneous003}
  ]
  particleSpawnScript [
   /Niagara/Modules/Spawn/Initialization/InitializeParticle(
    Color=&MakeLinearColorFromVectorAndFloat("Vector (RGB)"=<25 25 25>, "Float (Alpha)"=&UniformRangedFloat(Minimum=0.7, Maximum=1))
    Lifetime=&UniformRangedFloat(Minimum=0.4, Maximum=1.5)
    Mesh Scale=&VectorFromFloat(Value=&Lerp_Float(
      A=0.1
      B=0.3
      Alpha=&FloatFromCurve(
       FloatCurve=DataInterface() {propertyValues {Curve {keys [<at 0 1 RCIM_Constant>,<at 0.15 0 RCIM_Constant>]}}}
       CurveIndex=&RandomRangeFloat(Minimum=0, Maximum=1)
      )
     ))
    Module.Write Color=true
    Module.Write Lifetime=true
    Module.Write Mass=true
    Module.Write Material Random=true
    Module.Write Position=false
    Module.Write Scale=true
    Module.Write SpriteRotation=false
    Module.Write SpriteSize=false
    Position=&SimulationPosition()
    Sprite Size=<5 5>
   )
   &SkeletalMeshLocation(Bone Sampling Mode=::NewEnumerator0, MaskSpawnGroup=false, OverrideRandomSeed=false, Skeletal Mesh=$User.Skeletal Mesh)
   &AddVelocity(
    Cone Angle=145
    Cone Axis=<0 0 1>
    Module.UseConeFalloff=false
    Module.UseVelDistribution=false
    Override Seed=false
    Velocity Mode=::NewEnumerator2
    Velocity Speed=&Multiply_Float(A=0.25, B=&RandomRangeFloat(Minimum=500, Maximum=2000))
   ) {module=AddVelocity001}
   &AddVelocity(
    Velocity=&RandomRangeVector(Minimum=<400 400 0>, Maximum=<-400 -400 100>)
    Velocity Speed Scale=&RandomRangeFloat(Minimum=1.5, Maximum=3.5)
   ) {module=AddVelocity002}
   /Niagara/Modules/Spawn/Orientation/V2/InitialMeshOrientation(
    Rotation=&UniformRangedVector(Minimum=<0 0 0>, Maximum=<1 1 1>)
    Use Orientation Vector=true
    Use Rotation Vector=true
   ) {enabled=false}
   /Niagara/Modules/Spawn/Orientation/V3/InitialMeshOrientation(Rotation=&UniformRangedVector(Minimum=<0 0 0>, Maximum=<1 1 1>), Use Rotation Vector=true) {module=InitialMeshOrientation001}
  ]
  particleUpdateScript [
   &ParticleState()
   &AccelerationForce(Acceleration=&MakeVector(X=0, Y=0, Z=&RandomRangeFloat(Minimum=50, Maximum=100)))
   &CurlNoiseForce(Noise Strength=100, Pan Noise=false)
   &Drag(
    Drag=&Multiply_Float(A=1, B=&RandomRangeFloat(Minimum=8, Maximum=15))
    Use Linear Drag=true
    Use Rotational Drag=false
   )
   &SolveForcesAndVelocity(Clamp Velocity=false, Limit Acceleration=false)
   /Niagara/Modules/Update/Orientation/V2/UpdateMeshOrientation(
    Rotation Rate:NiagaraFloat=&FloatFromCurve(FloatCurve=DataInterface() {propertyValues {Curve {keys [{arriveTangent=-2.8458, interpMode=RCIM_Cubic, leaveTangent=-2.8458, tangentMode=RCTM_User, time=0, value=1}, <at 1 0.2 RCIM_Cubic>]}}})
    Rotation Vector:Vector3f=<2 0.4 0.1>
   ) {enabled=false}
   /Niagara/Modules/Update/Orientation/V3/UpdateMeshOrientation(
    Rotation Rate=&FloatFromCurve(FloatCurve=DataInterface() {propertyValues {Curve {keys [{arriveTangent=-2.8458, interpMode=RCIM_Cubic, leaveTangent=-2.8458, tangentMode=RCTM_User, time=0, value=1}, <at 1 0.2 RCIM_Cubic>]}}})
    Rotation Vector=<2 0.4 0.1>
   ) {module=UpdateMeshOrientation001}
   &ScaleMeshSize(Scale Factor=&VectorFromFloat(Value=&FloatFromCurve(FloatCurve=DataInterface() {propertyValues {Curve {keys [<at 0 0 RCIM_Cubic>,<at 0.25 2 RCIM_Cubic>,<at 1 0 RCIM_Cubic>]}}}, Scale Curve=0.5)))
   &ScaleColor(
    Scale Alpha=&FloatFromCurve(FloatCurve=DataInterface() {propertyValues {Curve {keys [<at 0.7 1 RCIM_Cubic>,<at 1 0>]}}})
    Scale RGB=&MakeVectorFromLinearColorRGB(Color=&ColorFromCurve(ColorCurve=DataInterface() {propertyValues {
        AlphaCurve {keys []}
        BlueCurve {keys [<at 0 2.8 RCIM_Cubic>,<at 0.4758 0.1>,<at 1 0 RCIM_Cubic>]}
        GreenCurve {keys [<at 0 5.7467 RCIM_Cubic>,<at 0.4758 0.185>,<at 1 0 RCIM_Cubic>]}
        RedCurve {keys [<at 0 8 RCIM_Cubic>,<at 0.4758 0.25>,<at 1 0 RCIM_Cubic>]}
       }}))
    ScaleA=true
    ScaleRGB=true
   )
  ]
  rendererClasses [&NiagaraMeshRendererProperties]
  renderers [{rendererClass=&NiagaraMeshRendererProperties, rendererIndex=0}]
 }
 Impact_Flare {
  emitterUpdateScript [
   &EmitterState()
   &SpawnBurst_Instantaneous(Spawn Count=1, Use Spawn Probability=false)
  ]
  particleSpawnScript [
   &InitializeParticle(
    Color=<rgba 2 1.258 0.2199 1>
    Lifetime=0.25
    Mesh Scale=&UniformRangedVector(Minimum=<1.25 1.25 1.25>, Maximum=<1.4 1.4 1.4>)
    Module.Write Color=true
    Module.Write Lifetime=true
    Module.Write Mass=true
    Module.Write Material Random=true
    Module.Write Position=false
    Module.Write Scale=true
    Module.Write SpriteRotation=false
    Module.Write SpriteSize=true
    Position=&SimulationPosition()
    Sprite Size=<200 175>
   )
   &CameraOffset(Camera Offset Amount=&RandomRangeFloat(Minimum=15, Maximum=25))
   &SystemLocation(MaskSpawnGroup=false, Offset=$System.FlarePos)
  ]
  particleUpdateScript [
   &ParticleState()
   &DynamicMaterialParameters(
    Index 0 Param 1=&FloatFromCurve(FloatCurve=DataInterface() {propertyValues {Curve {keys [<at 0 1 RCIM_Cubic>,<at 0.8 0 RCIM_Cubic>]}}})
    Index 0 Param 2=2
    Index 0 Param 3=1
    Index 0 Param 4=1
    Index 1 Param 1=&FloatFromCurve(FloatCurve=DataInterface() {propertyValues {Curve {keys [
         {arriveTangent=-2.3163, interpMode=RCIM_Cubic, time=0, value=2}
         {arriveTangent=-1.9019, interpMode=RCIM_Cubic, leaveTangent=-1.9019, tangentMode=RCTM_User, time=1, value=0.4}
        ]}}})
    Index 1 Param 2=0.4
    Index 1 Param 3=0.6
    Index 1 Param 4=1
    Index 2 Param 1=&FloatFromCurve(FloatCurve=DataInterface() {propertyValues {Curve {keys [<at 0.46 0>,<at 1 1>]}}})
    Index 2 Param 2=45
    Module.Param0WriteEnabled=true
    Module.Param0WriteEnabled1=false
    Module.Param0WriteEnabled2=true
    Module.Param1WriteEnabled=true
    Module.Param1WriteEnabled1=true
    Module.Param1WriteEnabled2=true
    Module.Param2WriteEnabled=true
    Module.Param2WriteEnabled1=true
    Module.Param2WriteEnabled2=true
    Module.Param3WriteEnabled=true
    Module.Param3WriteEnabled1=true
    Module.Param3WriteEnabled2=true
    Write Parameter Index 1=true
    Write Parameter Index 2=true
   )
   &ScaleColor(
    Scale Alpha=&FloatFromCurve(FloatCurve=DataInterface() {propertyValues {Curve {keys [<at 0 1>,<at 1 0>]}}})
    Scale RGB=&MakeVectorFromLinearColorRGB(Color=&LerpLinearColors(
      StartColor=<rgba 1 1 1 1>
      EndColor=<rgba 0 0 0 1>
      LerpFactor=&FloatFromCurve(FloatCurve=DataInterface() {propertyValues {Curve {keys [
           {arriveTangent=-2.5758, interpMode=RCIM_Cubic, leaveTangent=-2.5758, tangentMode=RCTM_User, time=0, value=1}
           {interpMode=RCIM_Cubic, tangentMode=RCTM_User, time=1, value=0}
          ]}}})
     ))
    ScaleA=true
    ScaleRGB=true
   )
   &ScaleColor(Scale Alpha=$System.TOD, ScaleA=true, ScaleRGB=false) {module=ScaleColor001}
  ]
  rendererClasses [&NiagaraSpriteRendererProperties]
  renderers [{rendererClass=&NiagaraSpriteRendererProperties}]
 }
 Light {
  emitterUpdateScript [
   &EmitterState()
   &SpawnBurst_Instantaneous(Use Spawn Probability=false, UseLoopCountLimit=false)
  ]
  particleSpawnScript [/Niagara/Modules/Spawn/Initialization/V2/InitializeParticle(Color Mode=::NewEnumerator0, Lifetime=0.4, UsePositionOffset=false, Write Material Random=true)]
  particleUpdateScript [
   &ParticleState()
   &SkeletalMeshLocation(
    Bone / Socket Index=11
    Bone Sampling Mode=::NewEnumerator4
    MaskSpawnGroup=false
    Orientation Sampling=::NewEnumerator1
    Skeletal Mesh=$User.Skeletal Mesh
   )
   &FN_Light_Attributes(
    Color=&ColorFromCurve(ColorCurve=DataInterface() {propertyValues {
       AlphaCurve {keys [<at 0 1 RCIM_Cubic>]}
       BlueCurve {keys [<at 0 1.8 RCIM_Cubic>]}
       GreenCurve {keys [<at 0 2.58 RCIM_Cubic>]}
       RedCurve {keys [<at 0 3 RCIM_Cubic>]}
      }})
    Create Color Attribute=true
    Create Exponent Attribute=true
    Create Radius Attribute=true
    Exponent=4
    Radius=&FloatFromCurve(
     FloatCurve=DataInterface() {propertyValues {Curve {keys [
         {arriveTangent=9.0725, interpMode=RCIM_Cubic, leaveTangent=9.0725, tangentMode=RCTM_User, time=0, value=0}
         {arriveTangent=-2.5249, interpMode=RCIM_Cubic, leaveTangent=-2.5249, tangentMode=RCTM_User, time=0.1408, value=0.9847}
         {arriveTangent=-0.0405, interpMode=RCIM_Cubic, leaveTangent=-0.0405, tangentMode=RCTM_User, time=1, value=0}
        ]}}}
     Scale Curve=350
    )
    Scale Color=&Multiply_Float(A=5, B=$System.TOD)
   )
  ]
  rendererClasses [&NiagaraLightRendererProperties]
  renderers [{rendererClass=&NiagaraLightRendererProperties, rendererIndex=0}]
 }
 Sparks {
  simTarget=GPUComputeSim
  emitterUpdateScript [
   &EmitterState()
   &SpawnBurst_Instantaneous(Spawn Count=30, Use Spawn Probability=false, UseLoopCountLimit=false)
  ]
  particleSpawnScript [
   &InitializeParticle(
    Color=&LerpLinearColors(
     StartColor=<rgba 5 4.4401 1.6406 1>
     EndColor=<rgba 4.6081 1.6406 5 1>
     LerpFactor=&MakeCustomFloatFromBool(Bool=&RandomBool(), True Float=1, False Float=0)
    )
    Lifetime Max=0.9
    Lifetime Min=0.25
    Lifetime Mode=::NewEnumerator1
    Sprite Size Max=<1.98 3.3>
    Sprite Size Min=<1.32 1.98>
    Sprite Size Mode=::NewEnumerator2
    UsePositionOffset=false
    Write Material Random=true
   ) {module=InitializeParticle001}
   &SphereLocation(
    MaskSpawnGroup=false
    Non Uniform Scale=<0.5 0.5 1>
    Offset=<0 0 160>
    Override Seed=false
    Sphere Radius=20
    Surface Expansion Mode=::NewEnumerator1
    Surface Only=true
    Surface Only Band Thickness=1
    UseNonUniformScale=true
    UseOffset=true
    UseSphereOrientation=false
   ) {module=SphereLocation001}
   &CameraOffset(Camera Offset Amount=15)
   &AddVelocityFromPoint(
    Module.Use Falloff Distance=false
    Origin Offset=<0 0 160>
    Velocity Strength=&UniformRangedFloat(Minimum=80, Maximum=800)
   )
  ]
  particleUpdateScript [
   &ParticleState()
   &CurlNoiseForce(
    Noise Frequency=22
    Noise Strength=&Multiply_Float(
     A=250
     B=&FloatFromCurve(FloatCurve=DataInterface() {propertyValues {Curve {keys [<at 0 0>,<at 0.1 1>,<at 1 1>]}}})
    )
    Pan Noise=false
   )
   &Drag(Drag=&UniformRangedFloat(Minimum=4, Maximum=8), Use Linear Drag=true, Use Rotational Drag=false)
   &SolveForcesAndVelocity(Clamp Velocity=false, Limit Acceleration=false)
   &ScaleSpriteSize(Scale Factor=&Vector2DFromFloat(Value=&FloatFromCurve(FloatCurve=DataInterface() {propertyValues {Curve {keys [<at 0 1>,<at 1 0>]}}})))
   &ScaleSpriteSizeBySpeed(Max Scale Factor=<0.5 3>, Velocity Threshold=500)
   &ScaleColor(
    Scale Alpha=&FloatFromCurve(FloatCurve=DataInterface() {propertyValues {Curve {keys [
         {interpMode=RCIM_Cubic, tangentMode=RCTM_User, time=0, value=0}
         {interpMode=RCIM_Cubic, tangentMode=RCTM_User, time=0.5, value=1}
         {tangentMode=RCTM_User, time=1, value=0}
        ]}}})
    ScaleA=true
    ScaleRGB=false
   )
  ]
  rendererClasses [&NiagaraSpriteRendererProperties]
  renderers [{rendererClass=&NiagaraSpriteRendererProperties}]
 }
}
summary {
 systemName=NS_Emote_Citadel_ArmorOff_Burst
 emitters [
  {emitterName=FakeLight, rendererClasses [&NiagaraSpriteRendererProperties]}
  {emitterName=Light, rendererClasses [&NiagaraLightRendererProperties]}
  {emitterName=Impact_Flare, rendererClasses [&NiagaraSpriteRendererProperties]}
  {emitterName=CoreFlare, rendererClasses [&NiagaraSpriteRendererProperties]}
  {emitterName=CoreFlare_MedHigh, rendererClasses [&NiagaraSpriteRendererProperties]}
  {emitterName=Corona, rendererClasses [&NiagaraSpriteRendererProperties]}
  {emitterName=Dark_Glow, rendererClasses [&NiagaraSpriteRendererProperties]}
  {emitterName=Fragments_Burst, rendererClasses [&NiagaraMeshRendererProperties]}
  {emitterName=Sparks, simTarget=GPUComputeSim, rendererClasses [&NiagaraSpriteRendererProperties]}
 ]
 userVariables [
  User.bIsFrontEnd {type=NiagaraBool}
  User.bIsFrontEndPreview {type=NiagaraBool}
  User.LoopsDelay {type=NiagaraFloat, defaultValue=0, description=""}
  User.Skeletal Mesh {
   type=NiagaraDataInterfaceSkeletalMesh
   defaultValue {
    struct=NiagaraExt_VariableValue_DataInterface
    value {dataInterface=&self:NiagaraDataInterfaceSkeletalMesh_0}
   }
  }
 ]
}
