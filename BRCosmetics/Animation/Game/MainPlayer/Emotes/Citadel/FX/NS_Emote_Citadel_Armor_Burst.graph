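# Niagara one-shot burst effect for the Citadel emote's armor, layering flashes, glows, a mesh shockwave, GPU sparks and two mesh-fragment bursts.
# Fragments fly out under drag and curl noise; one set takes User.SuperLevelColor and gets pulled toward a raised center late in life.
# Back glows and a light follow bones on User.Skeletal Mesh.
# Time-of-day values dim most glow, flash, light, shockwave and spark layers.
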NiagaraSystem /BRCosmetics/Animation/Game/MainPlayer/Emotes/Citadel/FX/NS_Emote_Citadel_Armor_Burst
dependencies {
 usedDataInterfaces [&NiagaraDataInterfaceCurve, &NiagaraDataInterfaceArrayInt32, &NiagaraDataInterfaceColorCurve]
 usedDynamicInputs []
 usedModules [
  &self:SystemSpawnScript.SystemScriptSource.SystemScriptGraph.NiagaraNodeAssignment_0.SetVariables_#1
  &SystemState
  &self:SystemSpawnScript.SystemScriptSource.SystemScriptGraph.NiagaraNodeAssignment_1.SetVariables_#2
 ]
}
emitters {
 CoreFlare_MedHigh {
  emitterUpdateScript [
   &EmitterState()
   &SpawnBurst_Instantaneous(Spawn Count=1, Use Spawn Probability=false)
  ]
  particleSpawnScript [
   /Niagara/Modules/Spawn/Initialization/InitializeParticle(
    Color=<rgba 5 4.125 2.5 0.1>
    Lifetime=0.35
    Module.Write Color=true
    Module.Write Lifetime=true
    Module.Write Mass=true
    Module.Write Material Random=true
    Module.Write Position=false
    Module.Write Scale=false
    Module.Write SpriteRotation=false
    Module.Write SpriteSize=true
    Position=&SimulationPosition()
    Sprite Size=&Vector2DFromFloat(Value=1500)
   )
   &CameraOffset(Camera Offset Amount=-20)
   &DynamicMaterialParameters(
    Index 0 Param 1=0.1
    Index 0 Param 2=1
    Index 0 Param 3=1
    Index 0 Param 4=1
    Module.Param0WriteEnabled=true
    Module.Param1WriteEnabled=true
    Module.Param2WriteEnabled=true
    Module.Param3WriteEnabled=true
   )
   &SphereLocation(
    MaskSpawnGroup=false
    Offset=<0 0 120>
    Override Seed=false
    Sphere Radius=0
    Surface Only=false
    UseNonUniformScale=false
    UseOffset=true
    UseSphereOrientation=false
   )
  ]
  particleUpdateScript [
   &ParticleState()
   &SolveForcesAndVelocity(Clamp Velocity=false, Limit Acceleration=false)
   &ScaleColor(Scale Alpha=0.7, ScaleA=true, ScaleRGB=true)
   &ScaleSpriteSize(Scale Factor=&Vector2DFromFloat(Value=&FloatFromCurve(FloatCurve=DataInterface() {propertyValues {Curve {keys [
          {arriveTangent=3310.2024, interpMode=RCIM_Cubic, time=0, value=0}
          {arriveTangent=0.7273, interpMode=RCIM_Cubic, leaveTangent=0.7273, time=0.1, value=0.3}
          {arriveTangent=-0.3333, interpMode=RCIM_Cubic, leaveTangent=-0.3333, time=0.55, value=0.4}
          {interpMode=RCIM_Cubic, leaveTangent=-1.6445, time=1, value=0}
         ]}}})))
   &ScaleColor(Scale Alpha=$System.TOD, ScaleA=true, ScaleRGB=false) {module=ScaleColor001}
  ]
  rendererClasses [&NiagaraSpriteRendererProperties]
  renderers [{rendererClass=&NiagaraSpriteRendererProperties}]
 }
 Dark_Glow {
  emitterUpdateScript [
   &EmitterState()
   &SpawnBurst_Instantaneous(Spawn Count=1, Use Spawn Probability=false)
  ]
  particleSpawnScript [
   /Niagara/Modules/Spawn/Initialization/V2/InitializeParticle(
    Color=<rgba 0.0288 0.0106 0.0312 2>
    Lifetime=0.22
    Sprite Size Mode=::NewEnumerator3
    Uniform Sprite Size=300
    UsePositionOffset=false
    Write Material Random=true
   ) {module=InitializeParticle001}
   &DynamicMaterialParameters(
    Index 0 Param 1=1
    Index 0 Param 2=1
    Index 0 Param 3=1
    Index 0 Param 4=1
    Module.Param0WriteEnabled=true
    Module.Param1WriteEnabled=true
    Module.Param2WriteEnabled=true
    Module.Param3WriteEnabled=true
   )
   &SystemLocation(MaskSpawnGroup=false, Offset=<0 0 90>)
  ]
  particleUpdateScript [
   &CameraOffset(Camera Offset Amount=-20)
   &ParticleState()
   &SolveForcesAndVelocity(Clamp Velocity=false, Limit Acceleration=false)
   &ScaleColor(
    Scale Alpha=&Multiply_Float(
     A=&FloatFromCurve(
      FloatCurve=DataInterface() {propertyValues {Curve {keys [<at 0 1>,<at 1 0>]}}}
      Scale Curve=&FN_FrontEndFloatModifier(Gameplay=1, Front End=0.5)
     )
     B=&FNTimeOfDayScalar(Night Scale=0.3, Morning Scale=0.2, Day Scale=0.5, Evening Scale=0.2)
    )
    ScaleA=true
    ScaleRGB=true
   )
   &ScaleSpriteSize(Scale Factor=&Vector2DFromFloat(Value=&FloatFromCurve(FloatCurve=DataInterface() {propertyValues {Curve {keys [<at 0 1>,<at 1 3>]}}})))
  ]
  rendererClasses [&NiagaraSpriteRendererProperties]
  renderers [{rendererClass=&NiagaraSpriteRendererProperties}]
 }
 Fragments_Burst {
  emitterUpdateScript [
   &EmitterState()
   &SpawnBurst_Instantaneous(
    Spawn Count=12
    Spawn Probability=&Lerp_Float(A=1, B=0, Alpha=$User.SuperLevel)
    Use Spawn Probability=false
    UseLoopCountLimit=false
   ) {module=SpawnBurst_Instantaneous003}
  ]
  particleSpawnScript [
   /Niagara/Modules/Spawn/Initialization/InitializeParticle(
    Color=$User.SuperLevelColor
    Lifetime=&RandomRangeFloat(Minimum=1.45, Maximum=1.6)
    Mesh Scale=&VectorFromFloat(Value=&RandomRangeFloat(Minimum=0.05, Maximum=0.08))
    Module.Write Color=true
    Module.Write Lifetime=true
    Module.Write Mass=true
    Module.Write Material Random=true
    Module.Write Position=false
    Module.Write Scale=true
    Module.Write SpriteRotation=false
    Module.Write SpriteSize=true
    Position=&SimulationPosition()
    Sprite Size=&Vector2DFromFloat(Value=20)
   )
   &ShapeLocation(MaskSpawnGroup=false, Offset=<0 0 90>, Offset Mode=::NewEnumerator0, Override Seed=false, Sphere Radius=20, UseOwnerScale=true) {module=ShapeLocation001}
   &AddVelocity(
    Velocity=&RandomRangeVector(Minimum=<400 400 400>, Maximum=<-400 -400 -400>)
    Velocity Speed Scale=&RandomRangeFloat(Minimum=2.5, Maximum=3.5)
   ) {module=AddVelocity002}
   &ApplyInitialForces()
   &InitialMeshOrientation(Rotation=&UniformRangedVector(Minimum=<0 0 0>, Maximum=<1 1 1>), Use Rotation Vector=true) {module=InitialMeshOrientation001}
   /SetVariables(Particles.MeshIndex=&RandomRangeInt(Minimum=0, Maximum=3)) {isSetParameters=true}
  ]
  particleUpdateScript [
   &ParticleState()
   &PointAttractionForce(
    Attraction Radius=200
    Attraction Strength=&Multiply_Float(
     A=&FloatFromCurve(FloatCurve=DataInterface() {propertyValues {Curve {keys [<at 0 0 RCIM_Constant>,<at 0.855 1 RCIM_Constant>]}}})
     B=1000
    )
    Attractor Position Offset:Vector3f=<0 0 90>
    Kill Within Radius=false
    Module.Use Falloff=true
    UseAttractorOffset=true
   )
   &CurlNoiseForce(Noise Strength=100, Pan Noise=false)
   &Drag(Drag=&Multiply_Float(A=1, B=13), Use Linear Drag=true, Use Rotational Drag=false)
   &SolveForcesAndVelocity(Clamp Velocity=false, Limit Acceleration=false)
   &UpdateMeshOrientation(
    Rotation Rate=&FloatFromCurve(
     FloatCurve=DataInterface() {propertyValues {Curve {keys [
         {arriveTangent=-2.8458, interpMode=RCIM_Cubic, leaveTangent=-2.8458, tangentMode=RCTM_User, time=0, value=1}
         {arriveTangent=-0.4597, interpMode=RCIM_Cubic, leaveTangent=-0.4597, tangentMode=RCTM_User, time=0.211, value=0.5515}
         <at 1 0.41 RCIM_Cubic>
        ]}}}
     Scale Curve=&RandomRangeFloat(Minimum=0.2, Maximum=0.15)
    )
    Rotation Vector=<2 0.4 0.1>
   ) {module=UpdateMeshOrientation001}
   &ScaleMeshSize(Scale Factor=&VectorFromFloat(Value=&FloatFromCurve(
      FloatCurve=DataInterface() {propertyValues {Curve {keys [
          <at 0 0 RCIM_Cubic>
          {arriveTangent=0.063, interpMode=RCIM_Cubic, leaveTangent=0.063, tangentMode=RCTM_User, time=0.15, value=1}
          {arriveTangent=0.0963, interpMode=RCIM_Cubic, leaveTangent=0.0963, tangentMode=RCTM_User, time=0.8519, value=1.0614}
          <at 1 0 RCIM_Cubic>
         ]}}}
      Scale Curve=0.5
     )))
   /SetVariables(Particles.Dynamic:Vector4f=&MakeVector4(
     X=2
     Y=1
     Z=&FloatFromCurve(FloatCurve=DataInterface() {propertyValues {Curve {keys [{interpMode=RCIM_Cubic, tangentMode=RCTM_User, time=0, value=0.525}, {interpMode=RCIM_Cubic, tangentMode=RCTM_User, time=0.8, value=-0.25}]}}})
     W=1
    )) {isSetParameters=true}
  ]
  rendererClasses [&NiagaraMeshRendererProperties]
  renderers [{rendererClass=&NiagaraMeshRendererProperties, rendererIndex=0}]
 }
 Fragments_Burst004 {
  emitterUpdateScript [
   &EmitterState()
   &SpawnBurst_Instantaneous(Spawn Count=20, Use Spawn Probability=false, UseLoopCountLimit=false) {module=SpawnBurst_Instantaneous003}
  ]
  particleSpawnScript [
   &InitializeParticle(
    Color=&MakeLinearColorFromVectorAndFloat("Vector (RGB)"=<25 25 25>, "Float (Alpha)"=&UniformRangedFloat(Minimum=0.7, Maximum=1))
    Lifetime=&UniformRangedFloat(Minimum=0.4, Maximum=1.5)
    Mesh Scale=&VectorFromFloat(Value=&Lerp_Float(
      A=0.1
      B=0.2
      Alpha=&FloatFromCurve(
       FloatCurve=DataInterface() {propertyValues {Curve {keys [<at 0 1 RCIM_Constant>,<at 0.15 0 RCIM_Constant>]}}}
       CurveIndex=&RandomRangeFloat(Minimum=0, Maximum=1)
      )
     ))
    Module.Write Color=true
    Module.Write Lifetime=true
    Module.Write Mass=true
    Module.Write Material Random=true
    Module.Write Position=false
    Module.Write Scale=true
    Module.Write SpriteRotation=false
    Module.Write SpriteSize=false
    Position=&SimulationPosition()
    Sprite Size=<5 5>
   )
   &ShapeLocation(MaskSpawnGroup=false, Offset=<0 0 90>, Offset Mode=::NewEnumerator0, Override Seed=false, Sphere Radius=20, UseOwnerScale=true) {module=ShapeLocation001}
   &AddVelocity(
    Velocity=&RandomRangeVector(Minimum=<400 400 400>, Maximum=<-400 -400 -400>)
    Velocity Speed Scale=&RandomRangeFloat(Minimum=2.5, Maximum=3.5)
   ) {module=AddVelocity002}
   &InitialMeshOrientation(Rotation=&UniformRangedVector(Minimum=<0 0 0>, Maximum=<1 1 1>), Use Rotation Vector=true) {module=InitialMeshOrientation001}
  ]
  particleUpdateScript [
   &ParticleState()
   &AccelerationForce(Acceleration=&MakeVector(X=0, Y=0, Z=&RandomRangeFloat(Minimum=50, Maximum=100)))
   &Drag(Drag=&Multiply_Float(A=1, B=13), Use Linear Drag=true, Use Rotational Drag=false)
   &CurlNoiseForce(Noise Strength=100, Pan Noise=false)
   &SolveForcesAndVelocity(Clamp Velocity=false, Limit Acceleration=false)
   &UpdateMeshOrientation(
    Rotation Rate=&FloatFromCurve(FloatCurve=DataInterface() {propertyValues {Curve {keys [{arriveTangent=-2.8458, interpMode=RCIM_Cubic, leaveTangent=-2.8458, tangentMode=RCTM_User, time=0, value=1}, <at 1 0.2 RCIM_Cubic>]}}})
    Rotation Vector=<2 0.4 0.1>
   ) {module=UpdateMeshOrientation001}
   &ScaleMeshSize(Scale Factor=&VectorFromFloat(Value=&FloatFromCurve(FloatCurve=DataInterface() {propertyValues {Curve {keys [<at 0 0 RCIM_Cubic>,<at 0.25 2 RCIM_Cubic>,<at 1 0 RCIM_Cubic>]}}}, Scale Curve=0.5)))
   &ScaleColor(
    Scale Alpha=&FloatFromCurve(FloatCurve=DataInterface() {propertyValues {Curve {keys [<at 0.7 1 RCIM_Cubic>,<at 1 0>]}}})
    Scale RGB=&MakeVectorFromLinearColorRGB(Color=&ColorFromCurve(ColorCurve=DataInterface() {propertyValues {
        AlphaCurve {keys []}
        BlueCurve {keys [<at 0 2.8 RCIM_Cubic>,<at 0.4758 0.1>,<at 1 0 RCIM_Cubic>]}
        GreenCurve {keys [<at 0 5.7467 RCIM_Cubic>,<at 0.4758 0.185>,<at 1 0 RCIM_Cubic>]}
        RedCurve {keys [<at 0 8 RCIM_Cubic>,<at 0.4758 0.25>,<at 1 0 RCIM_Cubic>]}
       }}))
    ScaleA=true
    ScaleRGB=true
   )
  ]
  rendererClasses [&NiagaraMeshRendererProperties]
  renderers [{rendererClass=&NiagaraMeshRendererProperties, rendererIndex=0}]
 }
 GlowBehind {
  emitterUpdateScript [
   &EmitterState()
   &SpawnBurst_Instantaneous(Spawn Count=3, Use Spawn Probability=false) {module=SpawnBurst_Instantaneous001}
  ]
  particleSpawnScript [
   &FN_SetSpawnGroup(Number of Spawn Groups=3) {module=FN_SetSpawnGroup001}
   /Niagara/Modules/Spawn/Initialization/V2/InitializeParticle(
    Color=<rgba 5 5 5 1>
    Lifetime=0.75
    Sprite Size Mode=::NewEnumerator3
    Uniform Sprite Size=95
    UsePositionOffset=false
    Write Material Random=true
   ) {module=InitializeParticle001}
  ]
  particleUpdateScript [
   &SkeletalMeshLocation(
    Bone / Socket Index=&SelectIntFromArray(Direct Array Index=&SpawnGroup(), Int Selection Array=DataInterface() {propertyValues {IntData [1,4,16]}})
    Bone Sampling Mode=::NewEnumerator4
    MaskSpawnGroup=false
    Skeletal Mesh=$User.Skeletal Mesh
   )
   &ParticleState()
   &CameraOffset(Camera Offset Amount=10)
   &DynamicMaterialParameters(
    Index 0 Param 1=2
    Index 0 Param 2=1
    Index 0 Param 3=&FloatFromCurve(FloatCurve=DataInterface() {propertyValues {Curve {keys [{interpMode=RCIM_Cubic, tangentMode=RCTM_User, time=0, value=0.525}, {interpMode=RCIM_Cubic, tangentMode=RCTM_User, time=0.8, value=-0.25}]}}})
    Module.Param0WriteEnabled=true
    Module.Param1WriteEnabled=true
    Module.Param2WriteEnabled=true
    Module.Param3WriteEnabled=true
   )
   &ScaleColor(
    Scale Alpha=&FloatFromCurve(
     FloatCurve=DataInterface() {propertyValues {Curve {keys [
         {arriveTangent=-0.3409, interpMode=RCIM_Cubic, leaveTangent=-0.3409, tangentMode=RCTM_User, time=0, value=0}
         <at 0.4 0.625 RCIM_Cubic>
         <at 1 0>
        ]}}}
     Scale Curve=0.05
    )
    Scale RGB=&MakeVectorFromLinearColorRGB(Color=&ColorFromCurve(ColorCurve=DataInterface() {propertyValues {
        AlphaCurve {keys [<at 0 1 RCIM_Cubic>]}
        BlueCurve {keys [<at 0 0.855>, {interpMode=RCIM_Cubic, leaveTangent=0.0148, time=1, value=0.771}]}
        GreenCurve {keys [<at 0 0.2821>, {interpMode=RCIM_Cubic, leaveTangent=-0.032, time=1, value=0.9389}]}
        RedCurve {keys [<at 0 0.7882>, {interpMode=RCIM_Cubic, leaveTangent=-0.8646, time=1, value=1}]}
       }}))
    ScaleA=true
    ScaleRGB=true
   )
   &ScaleSpriteSize(Uniform Curve Sprite Scale=DataInterface() {propertyValues {Curve {keys [
        {interpMode=RCIM_Cubic, tangentMode=RCTM_User, time=0, value=0.5}
        {interpMode=RCIM_Cubic, tangentMode=RCTM_User, time=0.2, value=0.95}
        {tangentMode=RCTM_User, time=1, value=0.5}
       ]}}})
   &ScaleColor(Scale Alpha=$System.TOD, ScaleA=true, ScaleRGB=false) {module=ScaleColor001}
  ]
  rendererClasses [&NiagaraSpriteRendererProperties]
  renderers [{rendererClass=&NiagaraSpriteRendererProperties}]
 }
 Light {
  emitterUpdateScript [
   &EmitterState()
   &SpawnBurst_Instantaneous(Use Spawn Probability=false, UseLoopCountLimit=false)
  ]
  particleSpawnScript [&InitializeParticle(Color Mode=::NewEnumerator0, Lifetime=0.4, UsePositionOffset=false, Write Material Random=true)]
  particleUpdateScript [
   &ParticleState()
   &SkeletalMeshLocation(
    Bone / Socket Index=11
    Bone Sampling Mode=::NewEnumerator4
    MaskSpawnGroup=false
    Orientation Sampling=::NewEnumerator1
    Skeletal Mesh=$User.Skeletal Mesh
   )
   &FN_Light_Attributes(
    Color=&ColorFromCurve(ColorCurve=DataInterface() {propertyValues {
       AlphaCurve {keys [<at 0 1 RCIM_Cubic>]}
       BlueCurve {keys [<at 0 1.8 RCIM_Cubic>]}
       GreenCurve {keys [<at 0 2.58 RCIM_Cubic>]}
       RedCurve {keys [<at 0 3 RCIM_Cubic>]}
      }})
    Create Color Attribute=true
    Create Exponent Attribute=true
    Create Radius Attribute=true
    Exponent=4
    Radius=&FloatFromCurve(
     FloatCurve=DataInterface() {propertyValues {Curve {keys [
         {arriveTangent=9.0725, interpMode=RCIM_Cubic, leaveTangent=9.0725, tangentMode=RCTM_User, time=0, value=0}
         {arriveTangent=-2.5249, interpMode=RCIM_Cubic, leaveTangent=-2.5249, tangentMode=RCTM_User, time=0.1408, value=0.9847}
         {arriveTangent=-0.0405, interpMode=RCIM_Cubic, leaveTangent=-0.0405, tangentMode=RCTM_User, time=1, value=0}
        ]}}}
     Scale Curve=350
    )
    Scale Color=&Multiply_Float(A=5, B=$System.TOD)
   )
  ]
  rendererClasses [&NiagaraLightRendererProperties]
  renderers [{rendererClass=&NiagaraLightRendererProperties, rendererIndex=0}]
 }
 Main_Flash {
  emitterUpdateScript [
   &EmitterState()
   &SpawnBurst_Instantaneous(Spawn Count=1, Use Spawn Probability=false)
  ]
  particleSpawnScript [
   &CameraOffset(Camera Offset Amount=20)
   &InitializeParticle(
    Color=<rgba 1 0.8141 0.3802 0.5>
    Lifetime=0.8
    Position Offset=<0 0 120>
    Position Offset Coordinate Space=::World
    Sprite Size Mode=::NewEnumerator3
    Uniform Sprite Size=800
    UsePositionOffset=true
    Write Material Random=true
   ) {module=InitializeParticle001}
  ]
  particleUpdateScript [
   &ParticleState()
   &ScaleColor(
    Scale Alpha=&FloatFromCurve(FloatCurve=DataInterface() {propertyValues {Curve {keys [{arriveTangent=-2.9556, interpMode=RCIM_Cubic, leaveTangent=-2.9556, tangentMode=RCTM_User, time=0, value=1}, <at 1 0>]}}})
    ScaleA=true
    ScaleRGB=true
   )
   &ScaleSpriteSize(Scale Factor=&Vector2DFromFloat(Value=&FloatFromCurve(FloatCurve=DataInterface() {propertyValues {Curve {keys [
          {arriveTangent=10.3957, interpMode=RCIM_Cubic, leaveTangent=10.3957, tangentMode=RCTM_User, time=0, value=0}
          <at 0.215 1 RCIM_Cubic>
          <at 1 0>
         ]}}})))
   &SolveForcesAndVelocity(Clamp Velocity=false, Limit Acceleration=false)
   &ScaleColor(Scale Alpha=$System.TOD, ScaleA=true, ScaleRGB=false) {module=ScaleColor001}
  ]
  rendererClasses [&NiagaraSpriteRendererProperties]
  renderers [{rendererClass=&NiagaraSpriteRendererProperties}]
 }
 Shockwave {
  emitterUpdateScript [
   &EmitterState()
   &SpawnBurst_Instantaneous(Use Spawn Probability=false, UseLoopCountLimit=false)
  ]
  particleSpawnScript [
   &InitializeParticle(
    Color=<rgba 1 1 1 0.1>
    Lifetime=0.5
    Mesh Scale Mode=::NewEnumerator3
    Position Offset=<0 0 90>
    UsePositionOffset=true
    Write Material Random=true
   )
   &InitialMeshOrientation(
    Mesh Axis To Orient=<0 0 1>
    Mesh Orientation Mode=::NewEnumerator2
    Orientation Vector=<0 0 1>
    Rotation=<0 0.25 0>
    Use Rotation Vector=true
   )
   &CameraOffset(Camera Offset Amount=44)
  ]
  particleUpdateScript [
   &ParticleState()
   &DynamicMaterialParameters(
    Index 0 Param 1=&FloatFromCurve(FloatCurve=DataInterface() {propertyValues {Curve {keys [<at 0 0.0333>,<at 1 0>]}}})
    Index 0 Param 2=&FloatFromCurve(FloatCurve=DataInterface() {propertyValues {Curve {keys [{tangentMode=RCTM_User, time=0, value=0.75}, {tangentMode=RCTM_User, time=1, value=0.5}]}}})
    Index 0 Param 3=&FloatFromCurve(FloatCurve=DataInterface() {propertyValues {Curve {keys [{tangentMode=RCTM_User, time=0, value=1}, {tangentMode=RCTM_User, time=1, value=0}]}}})
    Module.Param0WriteEnabled=true
    Module.Param1WriteEnabled=true
    Module.Param2WriteEnabled=true
    Module.Param3WriteEnabled=false
   )
   &ScaleColor(
    Scale RGB=&VectorFromFloat(Value=&FloatFromCurve(
      FloatCurve=DataInterface() {propertyValues {Curve {keys [<at 0 2 RCIM_Constant>, {interpMode=RCIM_Cubic, tangentMode=RCTM_Break, time=0.1951, value=0.8182}, {tangentMode=RCTM_User, time=1, value=0}]}}}
      Scale Curve=3
     ))
    ScaleA=false
    ScaleRGB=true
   ) {module=ScaleColor001}
   &ScaleMeshSize(Scale Factor=&VectorFromFloat(Value=&FloatFromCurve(
      FloatCurve=DataInterface() {propertyValues {Curve {keys [{arriveTangent=2.9365, interpMode=RCIM_Cubic, leaveTangent=3.0077, tangentMode=RCTM_Break, time=0, value=0.5}, <at 1 1.666>]}}}
      Scale Curve=1.5
     )))
   &ScaleColor(
    Linear Color Curve=DataInterface() {propertyValues {
      AlphaCurve {keys [<at 0 1 RCIM_Cubic>,<at 1 0 RCIM_Cubic>]}
      BlueCurve {keys [<at 0 1.145 RCIM_Cubic>,<at 0.9771 1>]}
      GreenCurve {keys [<at 0 2.0483 RCIM_Cubic>,<at 0.9771 0.1403>]}
      RedCurve {keys [<at 0 2.5 RCIM_Cubic>,<at 0.9771 0>]}
     }}
    Scale Mode=::NewEnumerator2
   )
   &ScaleColor(Scale Alpha=$System.TOD, ScaleA=true, ScaleRGB=false) {module=ScaleColor002}
  ]
  rendererClasses [&NiagaraMeshRendererProperties]
  renderers [{rendererClass=&NiagaraMeshRendererProperties, rendererIndex=0}]
 }
 Sparks {
  simTarget=GPUComputeSim
  emitterUpdateScript [
   &EmitterState()
   &SpawnBurst_Instantaneous(Spawn Count=25, Use Spawn Probability=false, UseLoopCountLimit=false)
  ]
  particleSpawnScript [
   &InitializeParticle(
    Color=&LerpLinearColors(
     StartColor=<rgba 8 7.1042 2.625 1>
     EndColor=<rgba 7.3729 2.625 8 1>
     LerpFactor=&MakeCustomFloatFromBool(Bool=&RandomBool(), True Float=1, False Float=0)
    )
    Lifetime Max=1.5
    Lifetime Min=0.6
    Lifetime Mode=::NewEnumerator1
    Sprite Size Max=<2 2>
    Sprite Size Min=<1.5 1.5>
    Sprite Size Mode=::NewEnumerator2
    UsePositionOffset=false
    Write Material Random=true
   ) {module=InitializeParticle001}
   &SphereLocation(
    MaskSpawnGroup=false
    Non Uniform Scale=<0.5 0.5 1>
    Offset=<0 0 90>
    Override Seed=false
    Sphere Radius=20
    Surface Expansion Mode=::NewEnumerator1
    Surface Only=true
    Surface Only Band Thickness=1
    UseNonUniformScale=true
    UseOffset=true
    UseSphereOrientation=false
   ) {module=SphereLocation001}
   &CameraOffset(Camera Offset Amount=15)
   &AddVelocityFromPoint(
    Module.Use Falloff Distance=false
    Origin Offset=<0 0 90>
    Velocity Strength=&UniformRangedFloat(Minimum=80, Maximum=800)
   )
  ]
  particleUpdateScript [
   &ParticleState()
   &CurlNoiseForce(
    Noise Frequency=22
    Noise Strength=&Multiply_Float(
     A=100
     B=&FloatFromCurve(FloatCurve=DataInterface() {propertyValues {Curve {keys [<at 0 0>,<at 0.1 1>,<at 1 1>]}}})
    )
    Pan Noise=false
   )
   &Drag(Drag=&UniformRangedFloat(Minimum=5, Maximum=10), Use Linear Drag=true, Use Rotational Drag=false)
   &SolveForcesAndVelocity(Clamp Velocity=false, Limit Acceleration=false)
   &ScaleSpriteSize(Scale Factor=&Vector2DFromFloat(Value=&FloatFromCurve(FloatCurve=DataInterface() {propertyValues {Curve {keys [<at 0 1>,<at 1 0>]}}})))
   &ScaleSpriteSizeBySpeed(Max Scale Factor=<1 10>, Velocity Threshold=500)
   &ScaleColor(
    Scale Alpha=&FloatFromCurve(
     FloatCurve=DataInterface() {propertyValues {Curve {keys [
         {interpMode=RCIM_Cubic, tangentMode=RCTM_User, time=0, value=0}
         {interpMode=RCIM_Cubic, tangentMode=RCTM_User, time=0.5, value=1}
         {tangentMode=RCTM_User, time=1, value=0}
        ]}}}
     Scale Curve=$System.TOD
    )
    ScaleA=true
    ScaleRGB=false
   )
  ]
  rendererClasses [&NiagaraSpriteRendererProperties]
  renderers [{rendererClass=&NiagaraSpriteRendererProperties}]
 }
}
summary {
 systemName=NS_Emote_Citadel_Armor_Burst
 emitters [
  {emitterName=Fragments_Burst, rendererClasses [&NiagaraMeshRendererProperties]}
  {emitterName=GlowBehind, rendererClasses [&NiagaraSpriteRendererProperties]}
  {emitterName=Main_Flash, rendererClasses [&NiagaraSpriteRendererProperties]}
  {emitterName=Fragments_Burst004, rendererClasses [&NiagaraMeshRendererProperties]}
  {emitterName=Light, rendererClasses [&NiagaraLightRendererProperties]}
  {emitterName=CoreFlare_MedHigh, rendererClasses [&NiagaraSpriteRendererProperties]}
  {emitterName=Sparks, simTarget=GPUComputeSim, rendererClasses [&NiagaraSpriteRendererProperties]}
  {emitterName=Shockwave, rendererClasses [&NiagaraMeshRendererProperties]}
  {emitterName=Dark_Glow, rendererClasses [&NiagaraSpriteRendererProperties]}
 ]
 userVariables [
  User.bIsFrontEnd {type=NiagaraBool}
  User.bIsFrontEndPreview {type=NiagaraBool}
  User.LoopsDelay {type=NiagaraFloat, defaultValue=0, description=""}
  User.Skeletal Mesh {
   type=NiagaraDataInterfaceSkeletalMesh
   defaultValue {
    struct=NiagaraExt_VariableValue_DataInterface
    value {dataInterface=&self:NiagaraDataInterfaceSkeletalMesh_0}
   }
  }
  User.SuperLevel {type=NiagaraFloat, defaultValue=0, description=""}
  User.SuperLevelColor {type=LinearColor, description="", defaultValue {struct=LinearColor, value=<rgba 1 1 1 1>}}
 ]
}
