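# Niagara system for the opening burst of the Citadel emote, a short warm flash of flares, glows, flames and light on the character.
# Single-sprite flare and corona emitters sit at System.FlarePos, with LOW and MedHigh named variants of some layers.
# Decal sprites, flames and glows use spawn groups to pick bones or sockets on User.Skeletal Mesh, which also anchors the light.
# Curves animate size, color and material parameters, and time-of-day scalars adjust intensity.
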NiagaraSystem /BRCosmetics/Animation/Game/MainPlayer/Emotes/Citadel/FX/NS_Emote_Citadel_Initial_Burst
dependencies {
 usedDataInterfaces [&NiagaraDataInterfaceCurve, &NiagaraDataInterfaceColorCurve, &NiagaraDataInterfaceArrayInt32]
 usedDynamicInputs []
 usedModules [
  &self:SystemSpawnScript.SystemScriptSource.SystemScriptGraph.NiagaraNodeAssignment_0.SetVariables_#1
  &self:SystemSpawnScript.SystemScriptSource.SystemScriptGraph.NiagaraNodeAssignment_1.SetVariables_#2
  &SystemState
  &self:SystemSpawnScript.SystemScriptSource.SystemScriptGraph.NiagaraNodeAssignment_2.SetVariables_#3
 ]
}
emitters {
 CoreFlare {
  emitterUpdateScript [
   &EmitterState()
   &SpawnBurst_Instantaneous(Spawn Count=1, Use Spawn Probability=false)
  ]
  particleSpawnScript [
   /Niagara/Modules/Spawn/Initialization/InitializeParticle(
    Color=<rgba 5 3.6875 1.25 0.15>
    Lifetime=0.35
    Module.Write Color=true
    Module.Write Lifetime=true
    Module.Write Mass=true
    Module.Write Material Random=true
    Module.Write Position=false
    Module.Write Scale=false
    Module.Write SpriteRotation=false
    Module.Write SpriteSize=true
    Position=&SimulationPosition()
    Sprite Size=&Vector2DFromFloat(Value=1500)
   )
   &CameraOffset(Camera Offset Amount=-30)
   &DynamicMaterialParameters(
    Index 0 Param 1=0.1
    Index 0 Param 2=1
    Index 0 Param 3=1
    Index 0 Param 4=1
    Module.Param0WriteEnabled=true
    Module.Param1WriteEnabled=true
    Module.Param2WriteEnabled=true
    Module.Param3WriteEnabled=true
   )
   &SphereLocation(
    MaskSpawnGroup=false
    Offset=$System.FlarePos
    Override Seed=false
    Sphere Radius=0
    Surface Only=false
    UseNonUniformScale=false
    UseOffset=true
    UseSphereOrientation=false
   )
  ]
  particleUpdateScript [
   &ParticleState()
   &SolveForcesAndVelocity(Clamp Velocity=false, Limit Acceleration=false)
   &ScaleColor(Scale Alpha=0.7, ScaleA=true, ScaleRGB=true)
   &ScaleSpriteSize(Scale Factor=&Vector2DFromFloat(Value=&FloatFromCurve(FloatCurve=DataInterface() {propertyValues {Curve {keys [
          {arriveTangent=3310.2024, interpMode=RCIM_Cubic, time=0, value=0}
          {arriveTangent=-15.1945, interpMode=RCIM_Cubic, leaveTangent=-3.0716, tangentMode=RCTM_Break, time=0.15, value=1}
          {interpMode=RCIM_Cubic, leaveTangent=-1.6445, time=1, value=0}
         ]}}})))
   &ScaleColor(Scale Alpha=$System.TOD, ScaleA=true, ScaleRGB=false) {module=ScaleColor001}
  ]
  rendererClasses [&NiagaraSpriteRendererProperties]
  renderers [{rendererClass=&NiagaraSpriteRendererProperties}]
 }
 CoreFlare_MedHigh {
  emitterUpdateScript [
   &EmitterState()
   &SpawnBurst_Instantaneous(Spawn Count=1, Use Spawn Probability=false)
  ]
  particleSpawnScript [
   &InitializeParticle(
    Color=<rgba 5 4.125 2.5 0.1>
    Lifetime=0.35
    Module.Write Color=true
    Module.Write Lifetime=true
    Module.Write Mass=true
    Module.Write Material Random=true
    Module.Write Position=false
    Module.Write Scale=false
    Module.Write SpriteRotation=false
    Module.Write SpriteSize=true
    Position=&SimulationPosition()
    Sprite Size=&Vector2DFromFloat(Value=1500)
   )
   &CameraOffset(Camera Offset Amount=-20)
   &DynamicMaterialParameters(
    Index 0 Param 1=0.1
    Index 0 Param 2=1
    Index 0 Param 3=1
    Index 0 Param 4=1
    Module.Param0WriteEnabled=true
    Module.Param1WriteEnabled=true
    Module.Param2WriteEnabled=true
    Module.Param3WriteEnabled=true
   )
   &SphereLocation(
    MaskSpawnGroup=false
    Offset=$System.FlarePos
    Override Seed=false
    Sphere Radius=0
    Surface Only=false
    UseNonUniformScale=false
    UseOffset=true
    UseSphereOrientation=false
   )
  ]
  particleUpdateScript [
   &ParticleState()
   &SolveForcesAndVelocity(Clamp Velocity=false, Limit Acceleration=false)
   &ScaleColor(Scale Alpha=0.7, ScaleA=true, ScaleRGB=true)
   &ScaleSpriteSize(Scale Factor=&Vector2DFromFloat(Value=&FloatFromCurve(FloatCurve=DataInterface() {propertyValues {Curve {keys [
          {arriveTangent=3310.2024, interpMode=RCIM_Cubic, time=0, value=0}
          {arriveTangent=0.7273, interpMode=RCIM_Cubic, leaveTangent=0.7273, time=0.1, value=0.3}
          {arriveTangent=-0.3333, interpMode=RCIM_Cubic, leaveTangent=-0.3333, time=0.55, value=0.4}
          {interpMode=RCIM_Cubic, leaveTangent=-1.6445, time=1, value=0}
         ]}}})))
   &ScaleColor(Scale Alpha=$System.TOD, ScaleA=true, ScaleRGB=false) {module=ScaleColor001}
  ]
  rendererClasses [&NiagaraSpriteRendererProperties]
  renderers [{rendererClass=&NiagaraSpriteRendererProperties}]
 }
 Corona {
  emitterUpdateScript [
   &EmitterState()
   &SpawnBurst_Instantaneous(Spawn Count=1, Use Spawn Probability=false)
  ]
  particleSpawnScript [
   &InitializeParticle(
    Color=<rgba 6 4.8 1.5 0.35>
    Lifetime=0.25
    Module.Write Color=true
    Module.Write Lifetime=true
    Module.Write Mass=true
    Module.Write Material Random=true
    Module.Write Position=false
    Module.Write Scale=false
    Module.Write SpriteRotation=false
    Module.Write SpriteSize=true
    Position=&SimulationPosition()
    Sprite Size=&Vector2DFromFloat(Value=350)
   )
   &CameraOffset(Camera Offset Amount=1)
   &SphereLocation(
    MaskSpawnGroup=false
    Offset=$System.FlarePos
    Override Seed=false
    Sphere Radius=0
    Surface Only=false
    UseNonUniformScale=false
    UseOffset=true
    UseSphereOrientation=false
   )
  ]
  particleUpdateScript [
   &ParticleState()
   &SolveForcesAndVelocity(Clamp Velocity=false, Limit Acceleration=false)
   &DynamicMaterialParameters(
    Index 0 Param 1=&FloatFromCurve(FloatCurve=DataInterface() {propertyValues {Curve {keys [{arriveTangent=-2.7401, interpMode=RCIM_Cubic, leaveTangent=-2.7401, tangentMode=RCTM_User, time=0, value=2}, <at 1 1 RCIM_Cubic>]}}})
    Index 0 Param 2=0.4
    Index 0 Param 3=0.6
    Index 0 Param 4=1
    Module.Param0WriteEnabled=true
    Module.Param1WriteEnabled=true
    Module.Param2WriteEnabled=true
    Module.Param3WriteEnabled=true
   )
   &ScaleColor(
    Scale Alpha=&FloatFromCurve(FloatCurve=DataInterface() {propertyValues {Curve {keys [
         <at 0 0 RCIM_Cubic>
         {arriveTangent=-0.5786, interpMode=RCIM_Cubic, leaveTangent=-0.5786, tangentMode=RCTM_User, time=0.15, value=0.6}
         <at 1 0 RCIM_Cubic>
        ]}}})
    Scale RGB=<2 2 2>
    ScaleA=true
    ScaleRGB=true
   )
   &ScaleSpriteSize(Scale Factor=&Vector2DFromFloat(Value=&FloatFromCurve(FloatCurve=DataInterface() {propertyValues {Curve {keys [
          {arriveTangent=3310.2024, interpMode=RCIM_Cubic, time=0, value=0.1}
          {arriveTangent=1.4167, interpMode=RCIM_Cubic, leaveTangent=1.4167, time=0.4, value=0.8}
          {arriveTangent=0.3333, interpMode=RCIM_Cubic, leaveTangent=0.3333, time=0.6, value=0.95}
          {interpMode=RCIM_Cubic, leaveTangent=-1.6445, time=1, value=1}
         ]}}})))
   &ScaleColor(Scale Alpha=$System.TOD, ScaleA=true, ScaleRGB=false) {module=ScaleColor001}
  ]
  rendererClasses [&NiagaraSpriteRendererProperties]
  renderers [{rendererClass=&NiagaraSpriteRendererProperties}]
 }
 Decals {
  emitterSpawnScript [/SetVariables(Emitter.DepthFade=75, Emitter.VisibleDepthDistance=75) {isSetParameters=true}]
  emitterUpdateScript [
   &EmitterState()
   &SpawnBurst_Instantaneous(Spawn Count=25, Use Spawn Probability=false) {module=SpawnBurst_Instantaneous001}
  ]
  particleSpawnScript [
   &FN_SetSpawnGroup(Number of Spawn Groups=25) {module=FN_SetSpawnGroup001}
   /Niagara/Modules/Spawn/Initialization/V2/InitializeParticle(
    Color=/Niagara/DynamicInputs/TypeConversions/MakeLinearColorFromFloat(Value:NiagaraFloat=2)
    Lifetime=0.75
    Sprite Size Mode=::NewEnumerator3
    Uniform Sprite Size=35
    UsePositionOffset=false
    Write Material Random=true
   ) {module=InitializeParticle001}
  ]
  particleUpdateScript [
   &SkeletalMeshLocation(Bone / Socket Index=$Particles.SpawnGroup, Bone Sampling Mode=::NewEnumerator4, MaskSpawnGroup=false, Skeletal Mesh=$User.Skeletal Mesh)
   &ParticleState()
   &CameraOffset(Camera Offset Amount=60)
   &DynamicMaterialParameters(
    Index 0 Param 1=2
    Index 0 Param 2=1
    Index 0 Param 3=&FloatFromCurve(FloatCurve=DataInterface() {propertyValues {Curve {keys [{interpMode=RCIM_Cubic, tangentMode=RCTM_User, time=0, value=-0.5}, {interpMode=RCIM_Cubic, tangentMode=RCTM_User, time=1, value=0.5}]}}})
    Index 0 Param 4=0.2
    Module.Param0WriteEnabled=true
    Module.Param1WriteEnabled=true
    Module.Param2WriteEnabled=true
    Module.Param3WriteEnabled=true
   ) {module=DynamicMaterialParameters001}
   &ScaleColor(
    Scale Alpha=&FloatFromCurve(FloatCurve=DataInterface() {propertyValues {Curve {keys [
         {arriveTangent=3, interpMode=RCIM_Cubic, leaveTangent=3, time=0.1, value=0}
         {arriveTangent=1.9398, interpMode=RCIM_Cubic, leaveTangent=1.9398, tangentMode=RCTM_User, time=0.2, value=0.3}
         <at 0.3 1>
         {arriveTangent=-3.5366, interpMode=RCIM_Cubic, leaveTangent=-3.5366, tangentMode=RCTM_User, time=0.5, value=0.95}
         <at 0.8 0>
        ]}}})
    Scale RGB=&MakeVectorFromLinearColorRGB(Color=&ColorFromCurve(ColorCurve=DataInterface() {propertyValues {
        AlphaCurve {keys [<at 0 1 RCIM_Cubic>]}
        BlueCurve {keys [{arriveTangent=-0.5763, interpMode=RCIM_Cubic, time=0, value=0.0104}, <at 0.3298 0.6>]}
        GreenCurve {keys [{arriveTangent=-0.5362, interpMode=RCIM_Cubic, time=0, value=0.7196}, <at 0.3298 0.8533>]}
        RedCurve {keys [<at 0 1 RCIM_Cubic>,<at 0.3298 1>]}
       }}))
    ScaleA=true
    ScaleRGB=true
   )
   &ScaleColor(
    Scale Alpha=$System.TOD
    Scale RGB=&VectorFromFloat(Value=&FloatFromCurve(
      FloatCurve=DataInterface() {propertyValues {Curve {keys [
          <at 0 -0.25>
          {arriveTangent=39.28, interpMode=RCIM_Cubic, leaveTangent=29.7857, tangentMode=RCTM_Break, time=0.1, value=3}
          <at 0.1828 20.3933>
          <at 0.2586 3>
          {arriveTangent=-2.1039, interpMode=RCIM_Cubic, leaveTangent=-2.1039, tangentMode=RCTM_User, time=1, value=0}
         ]}}}
      Scale Curve=4
     ))
    ScaleA=true
    ScaleRGB=true
   ) {module=ScaleColor001}
  ]
  rendererClasses [&NiagaraSpriteRendererProperties]
  renderers [{rendererClass=&NiagaraSpriteRendererProperties}]
 }
 Decals_LOW {
  emitterUpdateScript [
   &EmitterState()
   &SpawnBurst_Instantaneous(Spawn Count=23, Use Spawn Probability=false) {module=SpawnBurst_Instantaneous001}
  ]
  particleSpawnScript [
   &FN_SetSpawnGroup(Normalize Spawn Group=true, Number of Spawn Groups=23, Particles Per Group=1) {module=FN_SetSpawnGroup001}
   &InitializeParticle(
    Color=&MakeLinearColorFromFloat(Value:NiagaraFloat=1)
    Lifetime=0.4
    Sprite Size Mode=::NewEnumerator3
    Uniform Sprite Size=100
    UsePositionOffset=false
    Write Material Random=true
   ) {module=InitializeParticle001}
   &CameraOffset(Camera Offset Amount=30)
  ]
  particleUpdateScript [
   &SkeletalMeshLocation(Bone / Socket Index=$Particles.SpawnGroup, Bone Sampling Mode=::NewEnumerator4, MaskSpawnGroup=false, Skeletal Mesh=$User.Skeletal Mesh)
   &ParticleState()
   &ScaleColor(
    Scale Alpha=&FloatFromCurve(
     FloatCurve=DataInterface() {propertyValues {Curve {keys [
         {arriveTangent=3, interpMode=RCIM_Cubic, leaveTangent=3, time=0.1, value=0}
         {arriveTangent=1.9398, interpMode=RCIM_Cubic, leaveTangent=1.9398, tangentMode=RCTM_User, time=0.2, value=0.3}
         <at 0.3 1>
         {arriveTangent=-3.5366, interpMode=RCIM_Cubic, leaveTangent=-3.5366, tangentMode=RCTM_User, time=0.5, value=0.95}
         <at 0.8 0>
        ]}}}
     Scale Curve=0.6
    )
    Scale RGB=&MakeVectorFromLinearColorRGB(Color=&ColorFromCurve(ColorCurve=DataInterface() {propertyValues {
        AlphaCurve {keys [<at 0 1 RCIM_Cubic>]}
        BlueCurve {keys [{arriveTangent=-0.5763, interpMode=RCIM_Cubic, time=0, value=0.0104}, <at 0.3298 0.55>]}
        GreenCurve {keys [{arriveTangent=-0.5362, interpMode=RCIM_Cubic, time=0, value=0.7196}, <at 0.3298 0.835>]}
        RedCurve {keys [<at 0 1 RCIM_Cubic>,<at 0.3298 1>]}
       }}))
    ScaleA=true
    ScaleRGB=true
   )
   &ScaleColor(
    Scale Alpha=$System.TOD
    Scale RGB=&VectorFromFloat(Value=&FloatFromCurve(FloatCurve=DataInterface() {propertyValues {Curve {keys [
          <at 0 1>
          {arriveTangent=39.28, interpMode=RCIM_Cubic, leaveTangent=29.7857, tangentMode=RCTM_Break, time=0.1, value=3}
          <at 0.1828 20.3933>
          <at 0.2586 3>
          {arriveTangent=-2.1039, interpMode=RCIM_Cubic, leaveTangent=-2.1039, tangentMode=RCTM_User, time=1, value=0}
         ]}}}))
    ScaleA=true
    ScaleRGB=true
   ) {module=ScaleColor001}
  ]
  rendererClasses [&NiagaraSpriteRendererProperties]
  renderers [{rendererClass=&NiagaraSpriteRendererProperties}]
 }
 Flames {
  emitterUpdateScript [
   &EmitterState()
   &SpawnBurst_Instantaneous(Spawn Count=24, Use Spawn Probability=false)
  ]
  particleSpawnScript [
   &InitializeParticle(
    Color=&MultiplyLinearColor(
     Linear Color=&RandomRangeLinearColor(Minimum=<rgba 1 0.75 0.4 0.025>, Maximum=<rgba 1 0.8958 0.75 0.025>)
     Scale Factor=&MakeFloat4FromVectorAndFloat("Vector (RGB)"=<1 1 1>, "Float (Alpha)"=&RandomRangeFloat(Minimum=0.1, Maximum=0.4))
    )
    Lifetime Max=1.4
    Lifetime Min=0.4
    Lifetime Mode=::NewEnumerator1
    Sprite Rotation Mode=::NewEnumerator2
    Sprite Size Mode=::NewEnumerator1
    Uniform Sprite Size Max=100
    Uniform Sprite Size Min=90
    UsePositionOffset=false
    Write Material Random=true
   ) {module=InitializeParticle001}
   &FN_SetSpawnGroup(Normalize Spawn Group=true, Number of Spawn Groups=24, Particles Per Group=1)
   &SkeletalMeshLocation(Bone / Socket Index=$Particles.SpawnGroup, Bone Sampling Mode=::NewEnumerator4, MaskSpawnGroup=false, Skeletal Mesh=$User.Skeletal Mesh)
   &AddVelocity(Module.Use Falloff Distance=false, Override Seed=false, Velocity Mode=::NewEnumerator1, Velocity Speed=150) {module=AddVelocity001}
   &AddVelocity(Velocity=&RandomRangeVector(Minimum=<50 50 0>, Maximum=<-50 -50 0>), Velocity Speed Scale=3)
   /SetVariables(Particles.RandomFrame=&RandomRangeInt(Minimum=0, Maximum=63)) {isSetParameters=true}
   &CameraOffset(Camera Offset Amount=&RandomRangeFloat(Minimum=15, Maximum=35))
  ]
  particleUpdateScript [
   &ParticleState()
   &SpriteRotationRate(Rotation Rate=&Multiply_Float(
     A=&RandomRangeFloat(Minimum=2, Maximum=-2)
     B=&FloatFromCurve(
      FloatCurve=DataInterface() {propertyValues {Curve {keys [
          {arriveTangent=-4.0107, interpMode=RCIM_Cubic, leaveTangent=-4.0107, tangentMode=RCTM_User, time=-0.0046, value=1.0113}
          {arriveTangent=-0.2496, interpMode=RCIM_Cubic, leaveTangent=-0.2496, tangentMode=RCTM_User, time=0.5, value=0.16}
         ]}}}
      Scale Curve=150
     )
    ))
   &Drag(Drag=&RandomRangeFloat(Minimum=5, Maximum=7), Use Linear Drag=true, Use Rotational Drag=false)
   &GravityForce(Gravity=<0 0 -50>)
   &SolveForcesAndVelocity(Clamp Velocity=false, Limit Acceleration=false)
   &ScaleColor(
    Color Value To Scale=$Particles.Color
    Scale Alpha=&FloatFromCurve(
     FloatCurve=DataInterface() {propertyValues {Curve {keys [{arriveTangent=-1.342, interpMode=RCIM_Cubic, leaveTangent=-1.342, tangentMode=RCTM_User, time=0.1, value=1}, <at 1 0 RCIM_Cubic>]}}}
     Scale Curve=&FNTimeOfDayScalar(Night Scale=0.7, Morning Scale=0.3, Evening Scale=0.5)
    )
    ScaleA=true
    ScaleRGB=true
   )
   &ScaleSpriteSize(Scale Factor=&Vector2DFromFloat(Value=&FloatFromCurve(FloatCurve=DataInterface() {propertyValues {Curve {keys [
          {arriveTangent=0.6051, interpMode=RCIM_Cubic, leaveTangent=0.6051, tangentMode=RCTM_User, time=0, value=0.7}
          {arriveTangent=0.0897, interpMode=RCIM_Cubic, leaveTangent=0.0897, tangentMode=RCTM_User, time=1, value=1}
         ]}}})))
   &FN_SubUVAnimation(End Frame=&Add_Integer(A=$Particles.RandomFrame, B=63), Start Frame=$Particles.RandomFrame, SubUV Loop Count=0.25)
   &CameraOffset(Camera Offset Amount=35) {module=CameraOffset001}
  ]
  rendererClasses [&NiagaraSpriteRendererProperties]
  renderers [{rendererClass=&NiagaraSpriteRendererProperties}]
 }
 Flare {
  emitterUpdateScript [
   &EmitterState()
   &SpawnBurst_Instantaneous(Spawn Count=1, Use Spawn Probability=false)
  ]
  particleSpawnScript [
   /Niagara/Modules/Spawn/Initialization/InitializeParticle(
    Color=<rgba 2 1.5182 0.844 1>
    Lifetime=0.25
    Mesh Scale=&UniformRangedVector(Minimum=<1.25 1.25 1.25>, Maximum=<1.4 1.4 1.4>)
    Module.Write Color=true
    Module.Write Lifetime=true
    Module.Write Mass=true
    Module.Write Material Random=true
    Module.Write Position=false
    Module.Write Scale=true
    Module.Write SpriteRotation=false
    Module.Write SpriteSize=true
    Position=&SimulationPosition()
    Sprite Size=<200 175>
   )
   &CameraOffset(Camera Offset Amount=&RandomRangeFloat(Minimum=15, Maximum=25))
   &SystemLocation(MaskSpawnGroup=false, Offset=$System.FlarePos)
  ]
  particleUpdateScript [
   &ParticleState()
   &DynamicMaterialParameters(
    Index 0 Param 1=&FloatFromCurve(FloatCurve=DataInterface() {propertyValues {Curve {keys [<at 0 1 RCIM_Cubic>,<at 0.8 0 RCIM_Cubic>]}}})
    Index 0 Param 2=2
    Index 0 Param 3=1
    Index 0 Param 4=1
    Index 1 Param 1=&FloatFromCurve(FloatCurve=DataInterface() {propertyValues {Curve {keys [
         {arriveTangent=-2.3163, interpMode=RCIM_Cubic, time=0, value=2}
         {arriveTangent=-1.9019, interpMode=RCIM_Cubic, leaveTangent=-1.9019, tangentMode=RCTM_User, time=1, value=0.4}
        ]}}})
    Index 1 Param 2=0.4
    Index 1 Param 3=0.6
    Index 1 Param 4=1
    Index 2 Param 1=&FloatFromCurve(FloatCurve=DataInterface() {propertyValues {Curve {keys [<at 0.46 0>,<at 1 1>]}}})
    Index 2 Param 2=45
    Module.Param0WriteEnabled=true
    Module.Param0WriteEnabled1=false
    Module.Param0WriteEnabled2=true
    Module.Param1WriteEnabled=true
    Module.Param1WriteEnabled1=true
    Module.Param1WriteEnabled2=true
    Module.Param2WriteEnabled=true
    Module.Param2WriteEnabled1=true
    Module.Param2WriteEnabled2=true
    Module.Param3WriteEnabled=true
    Module.Param3WriteEnabled1=true
    Module.Param3WriteEnabled2=true
    Write Parameter Index 1=true
    Write Parameter Index 2=true
   )
   &ScaleColor(
    Scale Alpha=&FloatFromCurve(FloatCurve=DataInterface() {propertyValues {Curve {keys [<at 0 1>,<at 1 0>]}}})
    Scale RGB=&MakeVectorFromLinearColorRGB(Color=&LerpLinearColors(
      StartColor=<rgba 1 1 1 1>
      EndColor=<rgba 0 0 0 1>
      LerpFactor=&FloatFromCurve(FloatCurve=DataInterface() {propertyValues {Curve {keys [
           {arriveTangent=-2.5758, interpMode=RCIM_Cubic, leaveTangent=-2.5758, tangentMode=RCTM_User, time=0, value=1}
           {interpMode=RCIM_Cubic, tangentMode=RCTM_User, time=1, value=0}
          ]}}})
     ))
    ScaleA=true
    ScaleRGB=true
   )
   &ScaleColor(Scale Alpha=$System.TOD, ScaleA=true, ScaleRGB=false) {module=ScaleColor001}
  ]
  rendererClasses [&NiagaraSpriteRendererProperties]
  renderers [{rendererClass=&NiagaraSpriteRendererProperties}]
 }
 GlowBehind {
  emitterUpdateScript [
   &EmitterState()
   &SpawnBurst_Instantaneous(Spawn Count=3, Spawn Time=0.1, Use Spawn Probability=false) {module=SpawnBurst_Instantaneous001}
  ]
  particleSpawnScript [
   &FN_SetSpawnGroup(Number of Spawn Groups=3) {module=FN_SetSpawnGroup001}
   /Niagara/Modules/Spawn/Initialization/V2/InitializeParticle(
    Color=<rgba 5 5 5 1>
    Lifetime=0.5
    Sprite Size Mode=::NewEnumerator3
    Uniform Sprite Size=95
    UsePositionOffset=false
    Write Material Random=true
   ) {module=InitializeParticle001}
  ]
  particleUpdateScript [
   &SkeletalMeshLocation(
    Bone / Socket Index=&SelectIntFromArray(Direct Array Index=&SpawnGroup(), Int Selection Array=DataInterface() {propertyValues {IntData [1,4,16]}})
    Bone Sampling Mode=::NewEnumerator4
    MaskSpawnGroup=false
    Skeletal Mesh=$User.Skeletal Mesh
   )
   &ParticleState()
   &CameraOffset(Camera Offset Amount=10)
   &DynamicMaterialParameters(
    Index 0 Param 1=2
    Index 0 Param 2=1
    Index 0 Param 3=&FloatFromCurve(FloatCurve=DataInterface() {propertyValues {Curve {keys [{interpMode=RCIM_Cubic, tangentMode=RCTM_User, time=0, value=0.525}, {interpMode=RCIM_Cubic, tangentMode=RCTM_User, time=0.8, value=-0.25}]}}})
    Module.Param0WriteEnabled=true
    Module.Param1WriteEnabled=true
    Module.Param2WriteEnabled=true
    Module.Param3WriteEnabled=true
   )
   &ScaleColor(
    Scale Alpha=&FloatFromCurve(
     FloatCurve=DataInterface() {propertyValues {Curve {keys [
         {arriveTangent=-0.3409, interpMode=RCIM_Cubic, leaveTangent=-0.3409, tangentMode=RCTM_User, time=0, value=0}
         <at 0.4 0.625 RCIM_Cubic>
         <at 1 0>
        ]}}}
     Scale Curve=0.05
    )
    Scale RGB=&MakeVectorFromLinearColorRGB(Color=&ColorFromCurve(ColorCurve=DataInterface() {propertyValues {
        AlphaCurve {keys [<at 0 1 RCIM_Cubic>]}
        BlueCurve {keys [<at 0 0.855>, {interpMode=RCIM_Cubic, leaveTangent=0.0148, time=1, value=0.771}]}
        GreenCurve {keys [<at 0 0.2821>, {interpMode=RCIM_Cubic, leaveTangent=-0.032, time=1, value=0.9389}]}
        RedCurve {keys [<at 0 0.7882>, {interpMode=RCIM_Cubic, leaveTangent=-0.8646, time=1, value=1}]}
       }}))
    ScaleA=true
    ScaleRGB=true
   )
   &ScaleSpriteSize(Uniform Curve Sprite Scale=DataInterface() {propertyValues {Curve {keys [
        {interpMode=RCIM_Cubic, tangentMode=RCTM_User, time=0, value=0.5}
        {interpMode=RCIM_Cubic, tangentMode=RCTM_User, time=0.2, value=0.95}
        {tangentMode=RCTM_User, time=1, value=0.5}
       ]}}})
   &ScaleColor(Scale Alpha=$System.TOD, ScaleA=true, ScaleRGB=false) {module=ScaleColor001}
  ]
  rendererClasses [&NiagaraSpriteRendererProperties]
  renderers [{rendererClass=&NiagaraSpriteRendererProperties}]
 }
 Light {
  emitterUpdateScript [
   &EmitterState()
   &SpawnBurst_Instantaneous(Use Spawn Probability=false, UseLoopCountLimit=false)
  ]
  particleSpawnScript [&InitializeParticle(Color Mode=::NewEnumerator0, Lifetime=0.4, UsePositionOffset=false, Write Material Random=true)]
  particleUpdateScript [
   &ParticleState()
   &SkeletalMeshLocation(
    Bone / Socket Index=2
    Bone Sampling Mode=::NewEnumerator4
    MaskSpawnGroup=false
    Orientation Sampling=::NewEnumerator1
    Skeletal Mesh=$User.Skeletal Mesh
   )
   &FN_Light_Attributes(
    Color=&ColorFromCurve(ColorCurve=DataInterface() {propertyValues {
       AlphaCurve {keys [<at 0 1 RCIM_Cubic>]}
       BlueCurve {keys [<at 0 1.8 RCIM_Cubic>]}
       GreenCurve {keys [<at 0 2.58 RCIM_Cubic>]}
       RedCurve {keys [<at 0 3 RCIM_Cubic>]}
      }})
    Create Color Attribute=true
    Create Exponent Attribute=true
    Create Radius Attribute=true
    Exponent=4
    Radius=&FloatFromCurve(
     FloatCurve=DataInterface() {propertyValues {Curve {keys [
         {arriveTangent=9.0725, interpMode=RCIM_Cubic, leaveTangent=9.0725, tangentMode=RCTM_User, time=0, value=0}
         {arriveTangent=-2.5249, interpMode=RCIM_Cubic, leaveTangent=-2.5249, tangentMode=RCTM_User, time=0.1408, value=0.9847}
         {arriveTangent=-0.0405, interpMode=RCIM_Cubic, leaveTangent=-0.0405, tangentMode=RCTM_User, time=1, value=0}
        ]}}}
     Scale Curve=350
    )
    Scale Color=&Multiply_Float(A=5, B=$System.TOD)
   )
  ]
  rendererClasses [&NiagaraLightRendererProperties]
  renderers [{rendererClass=&NiagaraLightRendererProperties, rendererIndex=0}]
 }
}
summary {
 systemName=NS_Emote_Citadel_Initial_Burst
 emitters [
  {emitterName=Light, rendererClasses [&NiagaraLightRendererProperties]}
  {emitterName=Flare, rendererClasses [&NiagaraSpriteRendererProperties]}
  {emitterName=CoreFlare, rendererClasses [&NiagaraSpriteRendererProperties]}
  {emitterName=CoreFlare_MedHigh, rendererClasses [&NiagaraSpriteRendererProperties]}
  {emitterName=Corona, rendererClasses [&NiagaraSpriteRendererProperties]}
  {emitterName=GlowBehind, rendererClasses [&NiagaraSpriteRendererProperties]}
  {emitterName=Decals, rendererClasses [&NiagaraSpriteRendererProperties]}
  {emitterName=Flames, rendererClasses [&NiagaraSpriteRendererProperties]}
  {emitterName=Decals_LOW, rendererClasses [&NiagaraSpriteRendererProperties]}
 ]
 userVariables [User.Skeletal Mesh {
   type=NiagaraDataInterfaceSkeletalMesh
   defaultValue {
    struct=NiagaraExt_VariableValue_DataInterface
    value {dataInterface=&self:NiagaraDataInterfaceSkeletalMesh_0}
   }
  }]
}
