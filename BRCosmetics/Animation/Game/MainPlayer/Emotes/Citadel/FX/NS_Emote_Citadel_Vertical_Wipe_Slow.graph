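# Niagara effect for the Citadel emote, a slow vertical wipe of warm glow, flames and fragments over the character.
# Decals and FakeLight pin sprites to 25 skeletal mesh bones, with a curve-animated material parameter on the decals.
# Flames and rising mesh fragments spawn from the body, backed by bone glows, a dark flash and a neck flare with particle light.
# Time-of-day scalars dim several layers; Beam and Decals_Low are disabled.
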NiagaraSystem /BRCosmetics/Animation/Game/MainPlayer/Emotes/Citadel/FX/NS_Emote_Citadel_Vertical_Wipe_Slow
dependencies {
 usedDataInterfaces [
  &NiagaraDataInterfaceCurve
  &NiagaraDataInterfaceColorCurve
  &NiagaraDataInterfaceVector2DCurve
  &NiagaraDataInterfaceArrayInt32
  &NiagaraDataInterfaceArrayFloat
 ]
 usedDynamicInputs []
 usedModules [
  &self:SystemSpawnScript.SystemScriptSource.SystemScriptGraph.NiagaraNodeAssignment_1.SetVariables_#1
  &SystemState
  &self:SystemSpawnScript.SystemScriptSource.SystemScriptGraph.NiagaraNodeAssignment_0.SetVariables_#2
 ]
}
emitters {
 Beam {
  bEnabled=false
  emitterUpdateScript [
   &EmitterState(
    Enable Distance Culling=true
    Max Distance Response=::SleepAndClearParticles
    MaxDistance=7000
    Module.bUseMaxDistance=true
    Module.bUseMinDistance=false
    Reset Age On Awaken=true
    Scalability Mode=::NewEnumerator1
   )
   &SpawnBurst_Instantaneous(Spawn Count=1, Spawn Time=0.2, Use Spawn Probability=false)
  ]
  particleSpawnScript [
   /Niagara/Modules/Spawn/Initialization/InitializeParticle(
    Color=<rgba 2 1.637 0.7188 0.025>
    Lifetime=0.65
    Mass=&RandomRangeFloat(Minimum=0.6, Maximum=1.2)
    Module.Write Color=true
    Module.Write Lifetime=true
    Module.Write Mass=false
    Module.Write Material Random=true
    Module.Write Position=false
    Module.Write Scale=true
    Module.Write SpriteRotation=false
    Module.Write SpriteSize=true
    Sprite Size=&Vector2DFromFloat(Value=&RandomRangeFloat(Minimum=120, Maximum=150))
   )
   &SphereLocation(
    MaskSpawnGroup=false
    Offset=<0 0 -10>
    Override Seed=false
    Sphere Radius=1
    Surface Only=false
    UseNonUniformScale=false
    UseOffset=true
    UseSphereOrientation=false
   )
   &DynamicMaterialParameters(
    Index 0 Param 1=220
    Index 0 Param 3=&RandomRangeFloat(Minimum=0.4, Maximum=0.6)
    Index 0 Param 4=1
    Module.Param0WriteEnabled=false
    Module.Param1WriteEnabled=false
    Module.Param2WriteEnabled=false
    Module.Param3WriteEnabled=false
   )
   &InitialMeshOrientation(Orientation Vector=<0 0 1>, Use Orientation Vector=true, Use Rotation Vector=false) {module=InitialMeshOrientation001}
  ]
  particleUpdateScript [
   &ParticleState(Kill Particles When Lifetime Has Elapsed=false, Let Infinitely Lived Particles Die When Emitter Deactivates=true)
   &ScaleColor(
    Scale Alpha=&FloatFromCurve(FloatCurve=DataInterface() {propertyValues {Curve {keys [<at 0 0>,<at 0.25 1>,<at 1 0>]}}})
    ScaleA=true
    ScaleRGB=true
   )
   &DynamicMaterialParameters(
    Index 0 Param 1=&FloatFromCurve(FloatCurve=DataInterface() {propertyValues {Curve {keys [
         {arriveTangent=2175.405, interpMode=RCIM_Cubic, leaveTangent=2175.4055, tangentMode=RCTM_User, time=0, value=0}
         {interpMode=RCIM_Cubic, tangentMode=RCTM_User, time=0.25, value=220}
         {interpMode=RCIM_Cubic, tangentMode=RCTM_User, time=1, value=0}
        ]}}})
    Module.Param0WriteEnabled=true
    Module.Param1WriteEnabled=false
    Module.Param2WriteEnabled=false
    Module.Param3WriteEnabled=false
   ) {module=DynamicMaterialParameters001}
   &ScaleColor(Scale Alpha=$System.TOD, ScaleA=true, ScaleRGB=false) {module=ScaleColor001}
  ]
  rendererClasses [&NiagaraMeshRendererProperties]
  renderers [{rendererClass=&NiagaraMeshRendererProperties, rendererIndex=0}]
 }
 Dark_Glow {
  emitterUpdateScript [
   &EmitterState()
   &SpawnBurst_Instantaneous(Spawn Count=1, Spawn Time=0.2, Use Spawn Probability=false)
  ]
  particleSpawnScript [
   /Niagara/Modules/Spawn/Initialization/V2/InitializeParticle(
    Color=<rgba 0.0312 0.0226 0.0106 1.5>
    Lifetime=0.22
    Sprite Size Mode=::NewEnumerator3
    Uniform Sprite Size=300
    UsePositionOffset=false
    Write Material Random=true
   ) {module=InitializeParticle001}
   &DynamicMaterialParameters(
    Index 0 Param 1=1
    Index 0 Param 2=1
    Index 0 Param 3=1
    Index 0 Param 4=1
    Module.Param0WriteEnabled=true
    Module.Param1WriteEnabled=true
    Module.Param2WriteEnabled=true
    Module.Param3WriteEnabled=true
   )
   &SystemLocation(MaskSpawnGroup=false, Offset=<0 0 90>)
  ]
  particleUpdateScript [
   &CameraOffset(Camera Offset Amount=-20)
   &ParticleState()
   &SolveForcesAndVelocity(Clamp Velocity=false, Limit Acceleration=false)
   &ScaleColor(
    Scale Alpha=&Multiply_Float(
     A=&FloatFromCurve(
      FloatCurve=DataInterface() {propertyValues {Curve {keys [<at 0 1>,<at 1 0>]}}}
      Scale Curve=&FN_FrontEndFloatModifier(Gameplay=1, Front End=0.5)
     )
     B=&FNTimeOfDayScalar(Night Scale=0.3, Morning Scale=0.05, Day Scale=0.5, Evening Scale=0.05)
    )
    ScaleA=true
    ScaleRGB=true
   )
   &ScaleSpriteSize(Scale Factor=&Vector2DFromFloat(Value=&FloatFromCurve(FloatCurve=DataInterface() {propertyValues {Curve {keys [<at 0 1>,<at 1 3>]}}})))
  ]
  rendererClasses [&NiagaraSpriteRendererProperties]
  renderers [{rendererClass=&NiagaraSpriteRendererProperties}]
 }
 Decals {
  emitterSpawnScript [/SetVariables(Emitter.DepthFade=75, Emitter.VisibleDepthDistance=75) {isSetParameters=true}]
  emitterUpdateScript [
   &EmitterState()
   &SpawnBurst_Instantaneous(Spawn Count=25, Use Spawn Probability=false) {module=SpawnBurst_Instantaneous001}
  ]
  particleSpawnScript [
   &FN_SetSpawnGroup(Number of Spawn Groups=25) {module=FN_SetSpawnGroup001}
   &InitializeParticle(
    Color=&MakeLinearColorFromVectorAndFloat("Vector (RGB)"=&VectorFromFloat(Value=2), "Float (Alpha)"=1)
    Lifetime=1.25
    Sprite Size Mode=::NewEnumerator3
    Uniform Sprite Size=35
    UsePositionOffset=false
    Write Material Random=true
   ) {module=InitializeParticle001}
  ]
  particleUpdateScript [
   &SkeletalMeshLocation(Bone / Socket Index=$Particles.SpawnGroup, Bone Sampling Mode=::NewEnumerator4, MaskSpawnGroup=false, Skeletal Mesh=$User.SkelMesh)
   &ParticleState()
   &CameraOffset(Camera Offset Amount=60)
   &DynamicMaterialParameters(
    Index 0 Param 1=2
    Index 0 Param 2=1
    Index 0 Param 3=&FloatFromCurve(FloatCurve=DataInterface() {propertyValues {Curve {keys [<at 0 -0.3>,<at 1 0.3>]}}})
    Module.Param0WriteEnabled=true
    Module.Param1WriteEnabled=true
    Module.Param2WriteEnabled=true
    Module.Param3WriteEnabled=true
   )
   &ScaleColor(
    Scale Alpha=&FloatFromCurve(FloatCurve=DataInterface() {propertyValues {Curve {keys [
         <at 0 0 RCIM_Cubic>
         {arriveTangent=1.9398, interpMode=RCIM_Cubic, leaveTangent=1.9398, tangentMode=RCTM_User, time=0.3, value=0.3}
         <at 0.33 1>
         {arriveTangent=-3.5366, interpMode=RCIM_Cubic, leaveTangent=-3.5366, tangentMode=RCTM_User, time=0.65, value=1}
         <at 1 0>
        ]}}})
    Scale RGB=&MakeVectorFromLinearColorRGB(Color=&ColorFromCurve(ColorCurve=DataInterface() {propertyValues {
        AlphaCurve {keys [<at 0 1 RCIM_Cubic>]}
        BlueCurve {keys [<at 0 0.5 RCIM_Cubic>,<at 0.3298 0.65>,<at 1 0 RCIM_Cubic>]}
        GreenCurve {keys [<at 0 0.5 RCIM_Cubic>,<at 0.3298 0.9008>,<at 1 0.7167 RCIM_Cubic>]}
        RedCurve {keys [<at 0 0.5 RCIM_Cubic>,<at 0.3298 1>,<at 1 1 RCIM_Cubic>]}
       }}))
    ScaleA=true
    ScaleRGB=true
   )
   &ScaleColor(
    Scale RGB=&VectorFromFloat(Value=&FloatFromCurve(FloatCurve=DataInterface() {propertyValues {Curve {keys [
          <at 0 1>
          {arriveTangent=3.0581, interpMode=RCIM_Cubic, leaveTangent=3.0581, tangentMode=RCTM_User, time=0.0516, value=1}
          {arriveTangent=39.28, interpMode=RCIM_Cubic, leaveTangent=29.7857, tangentMode=RCTM_Break, time=0.0816, value=3}
          <at 0.65 7>
          {arriveTangent=-5.1937, interpMode=RCIM_Cubic, leaveTangent=-5.1937, tangentMode=RCTM_User, time=0.7409, value=1}
          <at 1 0.5 RCIM_Cubic>
         ]}}}))
    ScaleA=true
    ScaleRGB=true
   ) {module=ScaleColor001}
   &ScaleColor(Scale RGB=&VectorFromFloat(Value=$System.TOD), ScaleA=true, ScaleRGB=true) {module=ScaleColor002}
  ]
  rendererClasses [&NiagaraSpriteRendererProperties]
  renderers [{rendererClass=&NiagaraSpriteRendererProperties}]
 }
 Decals_Low {
  bEnabled=false
  emitterUpdateScript [
   &EmitterState()
   &SpawnBurst_Instantaneous(Spawn Count=17, Use Spawn Probability=false) {module=SpawnBurst_Instantaneous001}
  ]
  particleSpawnScript [
   &FN_SetSpawnGroup(Number of Spawn Groups=17) {module=FN_SetSpawnGroup001}
   &InitializeParticle(
    Color=<rgba 5 5 5 1>
    Lifetime=1.25
    Sprite Size Mode=::NewEnumerator3
    Uniform Sprite Size=35
    UsePositionOffset=false
    Write Material Random=true
   ) {module=InitializeParticle001}
  ]
  particleUpdateScript [
   &SkeletalMeshLocation(
    Bone / Socket Index=&SelectIntFromArray(
     Direct Array Index=$Particles.SpawnGroup
     Int Selection Array=DataInterface() {propertyValues {IntData [0,1,2,3,6,7,10,11,14,17,18,19,20,21,22,23,24]}}
    )
    Bone Sampling Mode=::NewEnumerator4
    MaskSpawnGroup=false
    Skeletal Mesh=$User.SkelMesh
   )
   &ParticleState()
   &CameraOffset(Camera Offset Amount=60)
   &DynamicMaterialParameters(
    Index 0 Param 2=1
    Index 0 Param 3=&FloatFromCurve(FloatCurve=DataInterface() {propertyValues {Curve {keys [<at 0.3 -0.3 RCIM_Cubic>,<at 0.6 0.3>]}}})
    Module.Param0WriteEnabled=true
    Module.Param1WriteEnabled=true
    Module.Param2WriteEnabled=true
    Module.Param3WriteEnabled=true
   )
   &ScaleColor(
    Scale Alpha=&FloatFromCurve(FloatCurve=DataInterface() {propertyValues {Curve {keys [
         <at 0 0 RCIM_Cubic>
         {arriveTangent=1.9398, interpMode=RCIM_Cubic, leaveTangent=1.9398, tangentMode=RCTM_User, time=0.3, value=0.3}
         <at 0.33 1>
         {arriveTangent=-3.5366, interpMode=RCIM_Cubic, leaveTangent=-3.5366, tangentMode=RCTM_User, time=0.65, value=1}
         <at 1 0>
        ]}}})
    Scale RGB=&MakeVectorFromLinearColorRGB(Color=&ColorFromCurve(ColorCurve=DataInterface() {propertyValues {
        AlphaCurve {keys [<at 0 1 RCIM_Cubic>]}
        BlueCurve {keys [<at 0 0.5 RCIM_Cubic>,<at 0.3298 0.9>,<at 1 0 RCIM_Cubic>]}
        GreenCurve {keys [<at 0 0.5 RCIM_Cubic>,<at 0.3298 0.9717>,<at 1 0.7167 RCIM_Cubic>]}
        RedCurve {keys [<at 0 0.5 RCIM_Cubic>,<at 0.3298 1>,<at 1 1 RCIM_Cubic>]}
       }}))
    ScaleA=true
    ScaleRGB=true
   ) {module=ScaleColor001}
   &ScaleSpriteSize(
    Scale Sprite Size Mode=::NewEnumerator0
    Uniform Scale Factor=&SelectFloatFromArray(
     Direct Array Index=$Particles.SpawnGroup
     Float Selection Array=DataInterface() {propertyValues {FloatData [1,1,1,1,1,1,1,1,1,1.25,1.25,1.25,1.25,1.25,1.25,1.25,1.25]}}
    )
   )
  ]
  rendererClasses [&NiagaraSpriteRendererProperties]
  renderers [{rendererClass=&NiagaraSpriteRendererProperties}]
 }
 FakeLight {
  emitterUpdateScript [
   &EmitterState()
   &SpawnBurst_Instantaneous(Spawn Count=25, Use Spawn Probability=false, UseLoopCountLimit=false)
  ]
  particleSpawnScript [
   &FN_SetSpawnGroup(Normalize Spawn Group=true, Number of Spawn Groups=25, Particles Per Group=1) {module=FN_SetSpawnGroup001}
   &InitializeParticle(Color=<rgba 8 6.2667 4 1>, Sprite Size Mode=::NewEnumerator3, Uniform Sprite Size=100, UsePositionOffset=false, Write Material Random=true)
   &CameraOffset(Camera Offset Amount=30)
  ]
  particleUpdateScript [
   &SkeletalMeshLocation(Bone / Socket Index=$Particles.SpawnGroup, Bone Sampling Mode=::NewEnumerator4, MaskSpawnGroup=false, Skeletal Mesh=$User.SkelMesh)
   &ParticleState()
   &ScaleColor(
    Scale Alpha=&FloatFromCurve(FloatCurve=DataInterface() {propertyValues {Curve {keys [
         {arriveTangent=0.0344, interpMode=RCIM_Cubic, leaveTangent=0.0344, tangentMode=RCTM_User, time=0, value=0}
         {arriveTangent=-2.5249, interpMode=RCIM_Cubic, leaveTangent=-2.5249, tangentMode=RCTM_User, time=0.3, value=0.9847}
         {arriveTangent=-0.0405, interpMode=RCIM_Cubic, leaveTangent=-0.0405, tangentMode=RCTM_User, time=1, value=0}
        ]}}})
    Scale RGB=&MakeVectorFromLinearColorRGB(Color=&ColorFromCurve(ColorCurve=DataInterface() {propertyValues {
        AlphaCurve {keys [<at 0 1 RCIM_Cubic>]}
        BlueCurve {keys [<at 0 6.4>,<at 0.1743 2.4>]}
        GreenCurve {keys [<at 0 7.3067>,<at 0.1743 4.44>]}
        RedCurve {keys [<at 0 8>,<at 0.1743 6>]}
       }}))
    ScaleA=true
    ScaleRGB=false
   )
   &ScaleColor(
    Scale Alpha=&FloatFromCurve(FloatCurve=DataInterface() {propertyValues {Curve {keys [
         <at 0 0 RCIM_Cubic>
         {arriveTangent=1.9398, interpMode=RCIM_Cubic, leaveTangent=1.9398, tangentMode=RCTM_User, time=0.3, value=0.3}
         <at 0.33 1>
         {arriveTangent=-3.5366, interpMode=RCIM_Cubic, leaveTangent=-3.5366, tangentMode=RCTM_User, time=0.65, value=1}
         <at 1 0>
        ]}}})
    Scale RGB=&MakeVectorFromLinearColorRGB(Color=&ColorFromCurve(ColorCurve=DataInterface() {propertyValues {
        AlphaCurve {keys [<at 0 1 RCIM_Cubic>]}
        BlueCurve {keys [<at 0 0.5 RCIM_Cubic>,<at 0.3298 0.65>,<at 1 0 RCIM_Cubic>]}
        GreenCurve {keys [<at 0 0.5 RCIM_Cubic>,<at 0.3298 0.9008>,<at 1 0.7167 RCIM_Cubic>]}
        RedCurve {keys [<at 0 0.5 RCIM_Cubic>,<at 0.3298 1>,<at 1 1 RCIM_Cubic>]}
       }}))
    ScaleA=true
    ScaleRGB=false
   ) {enabled=false, module=ScaleColor002}
   &ScaleColor(
    Scale RGB=&VectorFromFloat(Value=&FloatFromCurve(FloatCurve=DataInterface() {propertyValues {Curve {keys [
          <at 0 1>
          {arriveTangent=3.0581, interpMode=RCIM_Cubic, leaveTangent=3.0581, tangentMode=RCTM_User, time=0.0516, value=1}
          {arriveTangent=39.28, interpMode=RCIM_Cubic, leaveTangent=29.7857, tangentMode=RCTM_Break, time=0.0816, value=3}
          <at 0.65 7>
          {arriveTangent=-5.1937, interpMode=RCIM_Cubic, leaveTangent=-5.1937, tangentMode=RCTM_User, time=0.7409, value=1}
          <at 1 0.5 RCIM_Cubic>
         ]}}}))
    ScaleA=true
    ScaleRGB=true
   ) {enabled=false, module=ScaleColor003}
   &ScaleColor(Scale Alpha=$System.TOD, ScaleA=true, ScaleRGB=false) {module=ScaleColor001}
  ]
  rendererClasses [&NiagaraSpriteRendererProperties]
  renderers [{rendererClass=&NiagaraSpriteRendererProperties}]
 }
 Flames {
  emitterUpdateScript [
   &EmitterState(Life Cycle Mode=::NewEnumerator1, Loop Behavior=::NewEnumerator1, Loop Delay=0.5, Loop Duration=0.7, UseLoopDelay=true)
   &SpawnRate(SpawnRate=50, Use Spawn Probability=false)
  ]
  particleSpawnScript [
   &InitializeParticle(
    Color=&MultiplyLinearColor(
     Linear Color=&RandomRangeLinearColor(Minimum=<rgba 1 0.75 0.4 0.025>, Maximum=<rgba 1 0.8958 0.75 0.025>)
     Scale Factor=&MakeFloat4FromVectorAndFloat("Vector (RGB)"=<1 1 1>, "Float (Alpha)"=&RandomRangeFloat(Minimum=0.4, Maximum=0.1))
    )
    Lifetime Max=1.5
    Lifetime Min=0.8
    Lifetime Mode=::NewEnumerator1
    Sprite Rotation Mode=::NewEnumerator2
    Sprite Size Mode=::NewEnumerator1
    Uniform Sprite Size Max=100
    Uniform Sprite Size Min=90
    UsePositionOffset=false
    Write Material Random=true
   ) {module=InitializeParticle001}
   &FN_SetSpawnGroup(Normalize Spawn Group=true, Number of Spawn Groups=25, Particles Per Group=1)
   &SkeletalMeshLocation(Bone Sampling Mode=::NewEnumerator0, MaskSpawnGroup=false, OverrideRandomSeed=false, Skeletal Mesh=$User.SkelMesh)
   &AddVelocity(Module.Use Falloff Distance=false, Override Seed=false, Velocity Mode=::NewEnumerator1, Velocity Speed=150) {module=AddVelocity001}
   &AddVelocity(Velocity=&RandomRangeVector(Minimum=<50 0 0>, Maximum=<-50 0 0>))
   /SetVariables(Particles.RandomFrame=&RandomRangeInt(Minimum=0, Maximum=63)) {isSetParameters=true}
   &CameraOffset(Camera Offset Amount=&RandomRangeFloat(Minimum=15, Maximum=35))
  ]
  particleUpdateScript [
   &ParticleState()
   &SpriteRotationRate(Rotation Rate=&Multiply_Float(
     A=&RandomRangeFloat(Minimum=2, Maximum=-2)
     B=&FloatFromCurve(
      FloatCurve=DataInterface() {propertyValues {Curve {keys [
          {arriveTangent=-4.0107, interpMode=RCIM_Cubic, leaveTangent=-4.0107, tangentMode=RCTM_User, time=-0.0046, value=1.0113}
          {arriveTangent=-0.2496, interpMode=RCIM_Cubic, leaveTangent=-0.2496, tangentMode=RCTM_User, time=0.5, value=0.16}
         ]}}}
      Scale Curve=150
     )
    ))
   &Drag(Drag=&RandomRangeFloat(Minimum=5, Maximum=7), Use Linear Drag=true, Use Rotational Drag=false)
   &GravityForce(Gravity=&MakeVector(X=0, Y=0, Z=&RandomRangeFloat(Minimum=400, Maximum=50))) {module=GravityForce001}
   &SolveForcesAndVelocity(Clamp Velocity=false, Limit Acceleration=false)
   &ScaleColor(
    Color Value To Scale=$Particles.Color
    Scale Alpha=&FloatFromCurve(
     FloatCurve=DataInterface() {propertyValues {Curve {keys [{arriveTangent=-1.342, interpMode=RCIM_Cubic, leaveTangent=-1.342, tangentMode=RCTM_User, time=0.1, value=1}, <at 1 0 RCIM_Cubic>]}}}
     Scale Curve=&FNTimeOfDayScalar(Night Scale=0.7, Morning Scale=0.3, Evening Scale=0.5)
    )
    ScaleA=true
    ScaleRGB=true
   )
   &ScaleSpriteSize(Scale Factor=&Vector2DFromFloat(Value=&FloatFromCurve(FloatCurve=DataInterface() {propertyValues {Curve {keys [
          {arriveTangent=0.6051, interpMode=RCIM_Cubic, leaveTangent=0.6051, tangentMode=RCTM_User, time=0, value=0.7}
          {arriveTangent=0.0897, interpMode=RCIM_Cubic, leaveTangent=0.0897, tangentMode=RCTM_User, time=1, value=1}
         ]}}})))
   &FN_SubUVAnimation(End Frame=&Add_Integer(A=$Particles.RandomFrame, B=63), Start Frame=$Particles.RandomFrame, SubUV Loop Count=0.25)
   &CameraOffset(Camera Offset Amount=35) {module=CameraOffset001}
  ]
  rendererClasses [&NiagaraSpriteRendererProperties]
  renderers [{rendererClass=&NiagaraSpriteRendererProperties}]
 }
 Flare {
  emitterUpdateScript [
   &EmitterState()
   &SpawnBurst_Instantaneous(Spawn Time=0.5, Use Spawn Probability=false, UseLoopCountLimit=false)
  ]
  particleSpawnScript [
   &InitializeParticle(
    Lifetime=0.9
    Position Mode=::NewEnumerator1
    Sprite Size Mode=::NewEnumerator3
    Uniform Sprite Size=1
    UsePositionOffset=false
    Write Material Random=true
   )
   &SkeletalMeshLocation(
    Bone Sampling Mode=::NewEnumerator4
    MaskSpawnGroup=false
    Sampled Position Offset=<0 0 -50>
    Sampled Position Offset Coordinate Space:ENiagaraSkelMeshTransforms=::NewEnumerator1
    Skeletal Mesh=DataInterface() {propertyValues {
      DefaultMesh=null
      ExcludeBoneName=root
      OptionalComponentClass=null
      PreviewMesh=/BRCosmetics/Characters/Player/Male/Medium/Bodies/M_MED_Parallel_Comic/Meshes/M_MED_Parallel_Comic
      SkinningMode=SkinOnTheFly
      SoftSourceActor=null
      SourceMode=Default
      UvSetIndex=0
      WholeMeshLOD=-1
      bExcludeBone=false
      bReadDeformedGeometry=true
      bRequireCurrentFrameData=true
      ComponentTags []
      FilteredBones [neck_02]
      FilteredSockets []
      MeshUserParameter {parameter {name=None, typeDefHandle {registeredTypeIndex=1}, varData []}}
      SamplingRegions []
     }}
   )
   &CameraOffset(Camera Offset Amount=100)
  ]
  particleUpdateScript [
   &SolveForcesAndVelocity(Clamp Velocity=false, Limit Acceleration=false)
   &ParticleState()
   &ScaleSpriteSize(
    Non-Uniform Curve Scale=<500 200>
    Non-Uniform Curve Sprite Scale=DataInterface() {propertyValues {
      XCurve {keys [
        {interpMode=RCIM_Cubic, tangentMode=RCTM_User, time=0, value=0}
        {interpMode=RCIM_Cubic, tangentMode=RCTM_User, time=0.4, value=1}
        {tangentMode=RCTM_User, time=1, value=0}
       ]}
      YCurve {keys [
        {interpMode=RCIM_Cubic, tangentMode=RCTM_User, time=0, value=0}
        {interpMode=RCIM_Cubic, tangentMode=RCTM_User, time=0.8, value=1}
        {tangentMode=RCTM_User, time=1, value=0}
       ]}
     }}
    Scale Sprite Size Mode=::NewEnumerator3
   )
   &ScaleColor(
    Scale Alpha=&FloatFromCurve(
     FloatCurve=DataInterface() {propertyValues {Curve {keys [
         {interpMode=RCIM_Cubic, tangentMode=RCTM_User, time=0, value=0}
         {interpMode=RCIM_Cubic, tangentMode=RCTM_User, time=0.4, value=1}
         {tangentMode=RCTM_User, time=1, value=0}
        ]}}}
     Scale Curve=&Multiply_Float(A=0.5, B=$System.TOD)
    )
    Scale RGB=<0.8 0.8 0.8>
    ScaleA=true
    ScaleRGB=true
   )
   &Color(Color=&ColorFromCurve(ColorCurve=DataInterface() {propertyValues {
       AlphaCurve {keys [<at 0 1 RCIM_Cubic>,<at 1 0 RCIM_Cubic>]}
       BlueCurve {keys [<at 0 0.5 RCIM_Cubic>]}
       GreenCurve {keys [<at 0 0.8583 RCIM_Cubic>]}
       RedCurve {keys [<at 0 1 RCIM_Cubic>]}
      }}))
   /SetVariables(Particles.FlarePosition:NiagaraPosition=&ConvertVectorToPosition(
     Input Position=&Add_Vector(
      A=&ConvertPositionToVector(Input Position=$Particles.Position, Convert To Absolute Large World Vector=false)
      B=&MakeVector(
       X=0
       Y=0
       Z=&FloatFromCurve(
        FloatCurve=DataInterface() {propertyValues {Curve {keys [{tangentMode=RCTM_User, time=0, value=-0.5}, {tangentMode=RCTM_User, time=1, value=1}]}}}
        Scale Curve=100
       )
      )
     )
     Vector Is In Absolute Large World Space:NiagaraBool=false
    )) {isSetParameters=true}
   /SetVariables(Particles.LightPosition:NiagaraPosition=&ConvertVectorToPosition(
     Input Position=&Add_Vector(
      A=&ConvertPositionToVector(Input Position=$Particles.FlarePosition, Convert To Absolute Large World Vector=false)
      B=&MakeVector(
       X=0
       Y=0
       Z=&FloatFromCurve(
        FloatCurve=DataInterface() {propertyValues {Curve {keys [
            {interpMode=RCIM_Cubic, tangentMode=RCTM_User, time=0, value=0}
            {interpMode=RCIM_Cubic, tangentMode=RCTM_User, time=0.2, value=1.375}
            {arriveTangent=-1.7188, interpMode=RCIM_Cubic, leaveTangent=-1.7188, time=0.6, value=1}
            {tangentMode=RCTM_User, time=1, value=0}
           ]}}}
        Scale Curve=2
       )
      )
     )
     Vector Is In Absolute Large World Space:NiagaraBool=false
    )) {isSetParameters=true}
   &FN_Light_Attributes(
    Color=&MultiplyLinearColor(
     Linear Color=<rgba 1 0.8 0.4 1>
     Scale Factor=&MakeVector4_FromFloat(Float=&FloatFromCurve(FloatCurve=DataInterface() {propertyValues {Curve {keys [
           {arriveTangent=0.2066, interpMode=RCIM_Cubic, leaveTangent=0.2066, tangentMode=RCTM_User, time=0, value=0}
           {arriveTangent=0.8616, interpMode=RCIM_Cubic, leaveTangent=0.8616, tangentMode=RCTM_User, time=0.3, value=1}
           {arriveTangent=-0.1006, interpMode=RCIM_Cubic, leaveTangent=-0.1006, tangentMode=RCTM_User, time=1, value=0}
          ]}}}))
    )
    Create Color Attribute=true
    Create Exponent Attribute=true
    Create Radius Attribute=true
    Enable Particle Light Enable=true
    Exponent=3
    Radius=125
    Scale Color=5
   )
  ]
  rendererClasses [&NiagaraSpriteRendererProperties, &NiagaraLightRendererProperties]
  renderers [
   {rendererClass=&NiagaraSpriteRendererProperties}
   {rendererClass=&NiagaraLightRendererProperties, rendererIndex=1}
  ]
 }
 Fragments_Burst {
  emitterUpdateScript [
   &EmitterState(Life Cycle Mode=::NewEnumerator1, Loop Behavior=::NewEnumerator1, Loop Delay=0.5, Loop Duration=0.5, UseLoopDelay=true)
   &SpawnRate(SpawnRate=70, Use Spawn Probability=false)
  ]
  particleSpawnScript [
   /Niagara/Modules/Spawn/Initialization/InitializeParticle(
    Color=&MakeLinearColorFromVectorAndFloat("Vector (RGB)"=<25 25 25>, "Float (Alpha)"=&UniformRangedFloat(Minimum=0.7, Maximum=1))
    Lifetime=&UniformRangedFloat(Minimum=0.4, Maximum=1.5)
    Mesh Scale=&VectorFromFloat(Value=&Lerp_Float(
      A=0.1
      B=0.3
      Alpha=&FloatFromCurve(
       FloatCurve=DataInterface() {propertyValues {Curve {keys [<at 0 1 RCIM_Constant>,<at 0.15 0 RCIM_Constant>]}}}
       CurveIndex=&RandomRangeFloat(Minimum=0, Maximum=1)
      )
     ))
    Module.Write Color=true
    Module.Write Lifetime=true
    Module.Write Mass=true
    Module.Write Material Random=true
    Module.Write Position=false
    Module.Write Scale=true
    Module.Write SpriteRotation=false
    Module.Write SpriteSize=false
    Position=&SimulationPosition()
    Sprite Size=<5 5>
   )
   &SkeletalMeshLocation(MaskSpawnGroup=false, OverrideRandomSeed=false, Skeletal Mesh=$User.SkelMesh)
   &SphereLocation(
    MaskSpawnGroup=false
    Offset=<0 0 -50>
    Override Seed=false
    Sphere Radius=20
    Surface Only=false
    UseNonUniformScale=false
    UseOffset=true
    UseSphereOrientation=false
   ) {module=SphereLocation001}
   &AddVelocity(Coordinate Space=::World, Velocity=&UniformRangedVector(Minimum=<0 0 200>, Maximum=<0 0 1200>)) {module=AddVelocity003}
   &InitialMeshOrientation(Rotation=&UniformRangedVector(Minimum=<0 0 0>, Maximum=<1 1 1>), Use Rotation Vector=true) {module=InitialMeshOrientation001}
  ]
  particleUpdateScript [
   &ParticleState()
   &AccelerationForce(Acceleration=&RandomRangeVector(Minimum=<0 0 0>, Maximum=<0 0 200>)) {module=AccelerationForce001}
   &InheritSourceMovement(Applied Position Scale=<0.8 0.8 0.8>, ApplyForce=false, ApplyPosition=true, ApplyPositionFalloff=false)
   &Drag(Drag=&RandomRangeFloat(Minimum=5, Maximum=10), Use Linear Drag=true, Use Rotational Drag=false)
   &SolveForcesAndVelocity(Clamp Velocity=false, Limit Acceleration=false)
   &UpdateMeshOrientation(
    Rotation Rate=&FloatFromCurve(FloatCurve=DataInterface() {propertyValues {Curve {keys [{arriveTangent=-2.8458, interpMode=RCIM_Cubic, leaveTangent=-2.8458, tangentMode=RCTM_User, time=0, value=1}, <at 1 0.2 RCIM_Cubic>]}}})
    Rotation Vector=<2 0.4 0.1>
   ) {module=UpdateMeshOrientation001}
   &ScaleMeshSize(Scale Factor=&VectorFromFloat(Value=&FloatFromCurve(FloatCurve=DataInterface() {propertyValues {Curve {keys [<at 0 0 RCIM_Cubic>,<at 0.25 2 RCIM_Cubic>,<at 1 0 RCIM_Cubic>]}}}, Scale Curve=0.5)))
   &ScaleColor(
    Scale Alpha=&FloatFromCurve(FloatCurve=DataInterface() {propertyValues {Curve {keys [<at 0.7 1 RCIM_Cubic>,<at 1 0>]}}})
    Scale RGB=&MakeVectorFromLinearColorRGB(Color=&ColorFromCurve(ColorCurve=DataInterface() {propertyValues {
        AlphaCurve {keys []}
        BlueCurve {keys [<at 0 2.8 RCIM_Cubic>,<at 0.4758 0.1>,<at 1 0 RCIM_Cubic>]}
        GreenCurve {keys [<at 0 5.7467 RCIM_Cubic>,<at 0.4758 0.185>,<at 1 0 RCIM_Cubic>]}
        RedCurve {keys [<at 0 8 RCIM_Cubic>,<at 0.4758 0.25>,<at 1 0 RCIM_Cubic>]}
       }}))
    ScaleA=true
    ScaleRGB=true
   )
  ]
  rendererClasses [&NiagaraMeshRendererProperties]
  renderers [{rendererClass=&NiagaraMeshRendererProperties, rendererIndex=0}]
 }
 GlowBehind {
  emitterUpdateScript [
   &EmitterState()
   &SpawnBurst_Instantaneous(Spawn Count=3, Spawn Time=0.15, Use Spawn Probability=false) {module=SpawnBurst_Instantaneous001}
  ]
  particleSpawnScript [
   &FN_SetSpawnGroup(Number of Spawn Groups=3) {module=FN_SetSpawnGroup001}
   /Niagara/Modules/Spawn/Initialization/V2/InitializeParticle(
    Color=<rgba 5 5 5 1>
    Lifetime=1.25
    Sprite Size Mode=::NewEnumerator3
    Uniform Sprite Size=110
    UsePositionOffset=false
    Write Material Random=true
   ) {module=InitializeParticle001}
  ]
  particleUpdateScript [
   &SkeletalMeshLocation(
    Bone / Socket Index=&SelectIntFromArray(Direct Array Index=&SpawnGroup(), Int Selection Array=DataInterface() {propertyValues {IntData [1,4,16]}})
    Bone Sampling Mode=::NewEnumerator4
    MaskSpawnGroup=false
    Skeletal Mesh=$User.SkelMesh
   )
   &ParticleState()
   &CameraOffset(Camera Offset Amount=10)
   &DynamicMaterialParameters(
    Index 0 Param 1=2
    Index 0 Param 2=1
    Index 0 Param 3=&FloatFromCurve(FloatCurve=DataInterface() {propertyValues {Curve {keys [{interpMode=RCIM_Cubic, tangentMode=RCTM_User, time=0, value=0.525}, {interpMode=RCIM_Cubic, tangentMode=RCTM_User, time=0.8, value=-0.25}]}}})
    Module.Param0WriteEnabled=true
    Module.Param1WriteEnabled=true
    Module.Param2WriteEnabled=true
    Module.Param3WriteEnabled=true
   )
   &ScaleColor(
    Scale Alpha=&FloatFromCurve(
     FloatCurve=DataInterface() {propertyValues {Curve {keys [
         {arriveTangent=-0.3409, interpMode=RCIM_Cubic, leaveTangent=-0.3409, tangentMode=RCTM_User, time=0, value=0}
         <at 0.4 0.625 RCIM_Cubic>
         <at 1 0>
        ]}}}
     Scale Curve=0.05
    )
    Scale RGB=&MakeVectorFromLinearColorRGB(Color=&ColorFromCurve(ColorCurve=DataInterface() {propertyValues {
        AlphaCurve {keys [<at 0 1 RCIM_Cubic>]}
        BlueCurve {keys [<at 0 0.855>, {interpMode=RCIM_Cubic, leaveTangent=0.0148, time=1, value=0.771}]}
        GreenCurve {keys [<at 0 0.2821>, {interpMode=RCIM_Cubic, leaveTangent=-0.032, time=1, value=0.9389}]}
        RedCurve {keys [<at 0 0.7882>, {interpMode=RCIM_Cubic, leaveTangent=-0.8646, time=1, value=1}]}
       }}))
    ScaleA=true
    ScaleRGB=true
   )
   &ScaleColor(Scale RGB=&VectorFromFloat(Value=$System.TOD), ScaleA=true, ScaleRGB=true) {module=ScaleColor002}
   &ScaleSpriteSize(Uniform Curve Sprite Scale=DataInterface() {propertyValues {Curve {keys [
        {interpMode=RCIM_Cubic, tangentMode=RCTM_User, time=0, value=0.5}
        {interpMode=RCIM_Cubic, tangentMode=RCTM_User, time=0.2, value=0.95}
        {tangentMode=RCTM_User, time=1, value=0.5}
       ]}}})
   /SetVariables() {isSetParameters=true}
  ]
  rendererClasses [&NiagaraSpriteRendererProperties]
  renderers [{rendererClass=&NiagaraSpriteRendererProperties}]
 }
}
summary {
 systemName=NS_Emote_Citadel_Vertical_Wipe_Slow
 emitters [
  {emitterName=Decals, rendererClasses [&NiagaraSpriteRendererProperties]}
  {emitterName=Flare, rendererClasses [&NiagaraSpriteRendererProperties, &NiagaraLightRendererProperties]}
  {emitterName=GlowBehind, rendererClasses [&NiagaraSpriteRendererProperties]}
  {bEnabled=false, emitterName=Decals_Low, rendererClasses [&NiagaraSpriteRendererProperties]}
  {bEnabled=false, emitterName=Beam, rendererClasses [&NiagaraMeshRendererProperties]}
  {emitterName=Flames, rendererClasses [&NiagaraSpriteRendererProperties]}
  {emitterName=Fragments_Burst, rendererClasses [&NiagaraMeshRendererProperties]}
  {emitterName=Dark_Glow, rendererClasses [&NiagaraSpriteRendererProperties]}
  {emitterName=FakeLight, rendererClasses [&NiagaraSpriteRendererProperties]}
 ]
 userVariables [
  User.bIsFrontEnd {type=NiagaraBool}
  User.bIsFrontEndPreview {type=NiagaraBool}
  User.Skeletal Mesh {
   type=NiagaraDataInterfaceSkeletalMesh
   defaultValue {
    struct=NiagaraExt_VariableValue_DataInterface
    value {dataInterface=&self:NiagaraDataInterfaceSkeletalMesh_1}
   }
  }
  User.SkelMesh {
   type=NiagaraDataInterfaceSkeletalMesh
   defaultValue {
    struct=NiagaraExt_VariableValue_DataInterface
    value {dataInterface=&self:NiagaraDataInterfaceSkeletalMesh_0}
   }
  }
 ]
}
